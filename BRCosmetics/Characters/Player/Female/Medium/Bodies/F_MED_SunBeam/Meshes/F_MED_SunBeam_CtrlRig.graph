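# Control Rig for the SunBeam female character body that corrects the belt pivot bones as the thighs move.
# A spherical pose reader on each thigh weights fixed Euler rotation offsets on that side's front, side and back belt pivots.
# Parent constraints then blend the center front and back pivots evenly between their left and right neighbors.
# When isCrouching is true, both center pivots receive an extra fixed rotation.
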ControlRigBlueprint /BRCosmetics/Characters/Player/Female/Medium/Bodies/F_MED_SunBeam/Meshes/F_MED_SunBeam_CtrlRig
graphs [
 RigVMFunctionLibrary {links [], nodes []}
 RigVMModel {nodes [
   RigUnit_BeginExecution = &RigUnit_BeginExecution()
   n1 = /Script/ControlRig.RigUnit_SequenceExecution(ExecuteContext=@RigUnit_BeginExecution.ExecutePin) {is=struct RigVMUnitNode}
   n2 = &RigUnit_SphericalPoseReader(
    ActiveRegionScaleFactors=(PositiveWidth=0.5,NegativeWidth=4,PositiveHeight=4,NegativeHeight=4)
    ActiveRegionSize=0.1
    Debug=(bDrawDebug=false,DebugScale=25,DebugSegments=20,DebugThickness=0.25)
    DriverAxis=(X=1,Y=0,Z=0)
    DriverItem=(Type=Bone,Name="thigh_r")
    FalloffRegionScaleFactors=(PositiveWidth=1,NegativeWidth=1,PositiveHeight=1,NegativeHeight=1)
    FalloffSize=0.45
    FlipWidthScaling=false
    OptionalParentItem=(Type=Bone,Name="None")
    RotationOffset=(X=-90,Y=0,Z=0)
    ExecutePin=@n1.A
   )
   > &RigUnit_ModifyTransforms(
    ItemToModify=((Item=(Type=Bone,Name="dyn_belt_01_ft_r_pivot"),Transform=(Rotation=(X=0,Y=0,Z=0,W=1),Translation=(X=0,Y=0,Z=0),Scale3D=(X=1,Y=1,Z=1))))
    ItemToModify.0.Transform.Rotation=&RigVMFunction_MathQuaternionFromEuler(Euler=(X=0,Y=40,Z=0)).Result
    Weight=@n2.OutputParam
   )
   > &RigUnit_ModifyTransforms(
    ItemToModify=((Item=(Type=Bone,Name="dyn_belt_02_ft_r_pivot"),Transform=(Rotation=(X=0,Y=0,Z=0,W=1),Translation=(X=0,Y=0,Z=0),Scale3D=(X=1,Y=1,Z=1))))
    ItemToModify.0.Transform.Rotation=&RigVMFunction_MathQuaternionFromEuler(Euler=(X=15,Y=10,Z=0)).Result
    Weight=@n2.OutputParam
   )
   > &RigUnit_ModifyTransforms(
    ItemToModify=((Item=(Type=Bone,Name="dyn_belt_r_pivot"),Transform=(Rotation=(X=0,Y=0,Z=0,W=1),Translation=(X=0,Y=0,Z=0),Scale3D=(X=1,Y=1,Z=1))))
    ItemToModify.0.Transform.Rotation=&RigVMFunction_MathQuaternionFromEuler(Euler=(X=10,Y=10,Z=0)).Result
    Weight=@n2.OutputParam
   )
   > &RigUnit_ModifyTransforms(
    ItemToModify=((Item=(Type=Bone,Name="dyn_belt_bk_r_pivot"),Transform=(Rotation=(X=0,Y=0,Z=0,W=1),Translation=(X=0,Y=0,Z=0),Scale3D=(X=1,Y=1,Z=1))))
    ItemToModify.0.Transform.Rotation=&RigVMFunction_MathQuaternionFromEuler(Euler=(X=0,Y=5,Z=0)).Result
    Weight=@n2.OutputParam
   )
   n3 = &RigUnit_SphericalPoseReader(
    ActiveRegionScaleFactors=(PositiveWidth=0.5,NegativeWidth=4,PositiveHeight=4,NegativeHeight=4)
    ActiveRegionSize=0.1
    Debug=(bDrawDebug=false,DebugScale=25,DebugSegments=20,DebugThickness=0.25)
    DriverAxis=(X=-1,Y=0,Z=0)
    DriverItem=(Type=Bone,Name="thigh_l")
    FalloffRegionScaleFactors=(PositiveWidth=1,NegativeWidth=1,PositiveHeight=1,NegativeHeight=1)
    FalloffSize=0.45
    FlipHeightScaling=false
    FlipWidthScaling=true
    OptionalParentItem=(Type=Bone,Name="None")
    RotationOffset=(X=90,Y=0,Z=0)
    ExecutePin=@n1.B
   )
   > &RigUnit_ModifyTransforms(
    ItemToModify=((Item=(Type=Bone,Name="dyn_belt_01_ft_l_pivot"),Transform=(Rotation=(X=0,Y=0,Z=0,W=1),Translation=(X=0,Y=0,Z=0),Scale3D=(X=1,Y=1,Z=1))))
    ItemToModify.0.Transform.Rotation=&RigVMFunction_MathQuaternionFromEuler(Euler=(X=0,Y=-30,Z=0)).Result
    Weight=@n3.OutputParam
   )
   > &RigUnit_ModifyTransforms(
    ItemToModify=((Item=(Type=Bone,Name="dyn_belt_02_ft_l_pivot"),Transform=(Rotation=(X=0,Y=0,Z=0,W=1),Translation=(X=0,Y=0,Z=0),Scale3D=(X=1,Y=1,Z=1))))
    ItemToModify.0.Transform.Rotation=&RigVMFunction_MathQuaternionFromEuler(Euler=(X=0,Y=-25,Z=0)).Result
    Weight=@n3.OutputParam
   )
   > &RigUnit_ModifyTransforms(
    ItemToModify=((Item=(Type=Bone,Name="dyn_belt_l_pivot"),Transform=(Rotation=(X=0,Y=0,Z=0,W=1),Translation=(X=0,Y=0,Z=0),Scale3D=(X=1,Y=1,Z=1))))
    ItemToModify.0.Transform.Rotation=&RigVMFunction_MathQuaternionFromEuler(Euler=(X=25,Y=-20,Z=0)).Result
    Weight=@n3.OutputParam
   )
   > &RigUnit_ModifyTransforms(
    ItemToModify=((Item=(Type=Bone,Name="dyn_belt_bk_l_pivot"),Transform=(Rotation=(X=0,Y=0,Z=0,W=1),Translation=(X=0,Y=0,Z=0),Scale3D=(X=1,Y=1,Z=1))))
    ItemToModify.0.Transform.Rotation=&RigVMFunction_MathQuaternionFromEuler(Euler=(X=10,Y=-15,Z=0)).Result
    Weight=@n3.OutputParam
   )
   &RigUnit_ParentConstraint(
    AdvancedSettings=(RotationOrderForFilter=XZY)
    Child=(Type=Bone,Name="dyn_belt_ft_pivot")
    Parents=((Item=(Type=Bone,Name="dyn_belt_01_ft_l_pivot"),Weight=0.5),(Item=(Type=Bone,Name="dyn_belt_01_ft_r_pivot"),Weight=0.5))
    Weight=1
    ExecutePin=@n1.C
   )
   > n4 = &RigUnit_ParentConstraint(
    AdvancedSettings=(RotationOrderForFilter=XZY)
    Child=(Type=Bone,Name="dyn_belt_bk_pivot")
    Parents=((Item=(Type=Bone,Name="dyn_belt_bk_l_pivot"),Weight=0.5),(Item=(Type=Bone,Name="dyn_belt_bk_r_pivot"),Weight=0.5))
    Weight=1
   )
   Branch_2_RigVMUnitNode = &RigVMFunction_ControlFlowBranch(ExecuteContext=@n4.ExecutePin, Condition=RigVMVariableNode() {variable=isCrouching}.Value)
   &RigUnit_ModifyTransforms(
    ItemToModify=((Item=(Type=Bone,Name="dyn_belt_bk_pivot"),Transform=(Rotation=(X=0,Y=0,Z=0,W=1),Translation=(X=0,Y=0,Z=0),Scale3D=(X=1,Y=1,Z=1))))
    Weight=1
    ExecutePin=@Branch_2_RigVMUnitNode.True
    ItemToModify.0.Transform.Rotation=&RigVMFunction_MathQuaternionFromEuler(Euler=(X=0,Y=5,Z=0)).Result
   )
   > &RigUnit_ModifyTransforms(
    ItemToModify=((Item=(Type=Bone,Name="dyn_belt_ft_pivot"),Transform=(Rotation=(X=0,Y=0,Z=0,W=1),Translation=(X=0,Y=0,Z=0),Scale3D=(X=1,Y=1,Z=1))))
    Weight=1
    ItemToModify.0.Transform.Rotation=&RigVMFunction_MathQuaternionFromEuler(Euler=(X=5,Y=0,Z=0)).Result
   )
  ]}
]
hierarchy [
 root {loc=<0 0 0>, rot=<0 0 0 1>}
 pelvis {parent=root, loc=<0 -0.5301 98.0571>, rot=<0.0127 0.707 -0.0127 -0.707>}
 spine_01 {parent=pelvis, loc=<6.1208 -0.9328 0>, rot=<0 0 0.0576 -0.9983>}
 spine_02 {parent=spine_01, loc=<6.6022 -0.0153 0>, rot=<0 0 0.0022 -1>}
 spine_03 {parent=spine_02, loc=<6.4919 -0.0105 0>, rot=<0 0 0.0015 -1>}
 spine_04 {parent=spine_03, loc=<6.9596 -0.007 0>, rot=<0 0 0.0009 -1>}
 spine_05 {parent=spine_04, loc=<6.9981 -0.0017 0>, rot=<0 0 -0.0042 -1>}
 neck_01 {parent=spine_05, loc=<11.3138 0.1529 0>, rot=<0 0 -0.2003 -0.9797>}
 neck_02 {parent=neck_01, loc=<4.5627 0 0>, rot=<0 0 0.0504 -0.9987>}
 head {parent=neck_02, loc=<5.3914 0 0>, rot=<0 0 0.115 -0.9934>}
 clavicle_l {parent=spine_05, loc=<5.4831 -0.1053 -3.2946>, rot=<0.0938 -0.7388 0.0307 -0.6666>}
 upperarm_l {parent=clavicle_l, loc=<9.9486 0 0>, rot=<0.0371 -0.2992 -0.0705 -0.9509>}
 lowerarm_l {parent=upperarm_l, loc=<26.4793 0 0>, rot=<0 0 -0.2382 -0.9712>}
 hand_l {parent=lowerarm_l, loc=<22.7458 0 0>, rot=<-0.6747 -0.0182 0.0605 -0.7354>}
 thumb_01_l {parent=hand_l, loc=<1.3118 -0.3191 -2.459>, rot=<0.6546 -0.4042 0.0664 -0.6354>}
 thumb_02_l {parent=thumb_01_l, loc=<5.9322 -0.0253 0.1416>, rot=<0.0121 0.0046 0.2964 -0.955>}
 thumb_03_l {parent=thumb_02_l, loc=<2.2223 0.3682 0>, rot=<0 0 0.0172 -0.9999>}
 index_metacarpal_l {parent=hand_l, loc=<1.6535 0.0183 -2.96>, rot=<0.0158 -0.0124 -0.0173 -0.9996>}
 index_01_l {parent=index_metacarpal_l, loc=<8.3371 0.1718 -0.2461>, rot=<0.0363 0.0499 0.142 -0.9879>}
 index_02_l {parent=index_01_l, loc=<3.4919 -0.3004 0>, rot=<0 0 0.2948 -0.9556>}
 index_03_l {parent=index_02_l, loc=<2.5578 -0.123 0>, rot=<0 0 0.1968 -0.9804>}
 middle_metacarpal_l {parent=hand_l, loc=<2.1856 -0.1725 -0.5984>, rot=<0.006 0.0168 -0.0329 -0.9993>}
 middle_01_l {parent=middle_metacarpal_l, loc=<7.6374 0.346 -0.1508>, rot=<-0.0085 0.0328 0.1813 -0.9829>}
 middle_02_l {parent=middle_01_l, loc=<3.9108 0 0>, rot=<0 0 0.2808 -0.9598>}
 middle_03_l {parent=middle_02_l, loc=<2.6103 0.0517 0>, rot=<0 0 0.1446 -0.9895>}
 ring_metacarpal_l {parent=hand_l, loc=<2.2074 -0.3625 1.3638>, rot=<-0.0284 0.0722 -0.0458 -0.9959>}
 ring_01_l {parent=ring_metacarpal_l, loc=<7.2483 0.0329 -0.1426>, rot=<-0.0456 -0.0277 0.2951 -0.954>}
 ring_02_l {parent=ring_01_l, loc=<3.7208 0.4647 0>, rot=<0 0 0.1583 -0.9874>}
 ring_03_l {parent=ring_02_l, loc=<2.4152 0.114 0>, rot=<0 0 0.1504 -0.9886>}
 pinky_metacarpal_l {parent=hand_l, loc=<1.7616 -0.9753 3.3814>, rot=<-0.0397 0.0814 -0.069 -0.9935>}
 pinky_01_l {parent=pinky_metacarpal_l, loc=<7.1933 -0.4205 -0.2601>, rot=<-0.0487 -0.0063 0.2556 -0.9655>}
 pinky_02_l {parent=pinky_01_l, loc=<2.8274 0.0893 0>, rot=<0 0 0.2474 -0.9689>}
 pinky_03_l {parent=pinky_02_l, loc=<2.2046 0.0307 0>, rot=<0 0 0.125 -0.9922>}
 deform_wrist_dn_base_l {parent=hand_l, loc=<0 -1.5326 0>, rot=<0 0 0 -1>}
 deform_wrist_dn_l {parent=deform_wrist_dn_base_l, loc=<0 0 0>, rot=<0 0 0 1>}
 deform_wrist_up_base_l {parent=hand_l, loc=<0 2.3009 0>, rot=<0 0 0 -1>}
 deform_wrist_up_l {parent=deform_wrist_up_base_l, loc=<0 0 0>, rot=<0 0 0 1>}
 weapon_l {parent=hand_l, loc=<-0.1974 -3.4111 0.3811>, rot=<0 0 0 1>}
 hand_attach_l {parent=hand_l, loc=<8.5 -3 0>, rot=<0.7071 0.7071 0 0>}
 lowerarm_twist_02_l {parent=lowerarm_l, loc=<9.6199 0 0>, rot=<0 0 0 -1>}
 dyn_armTie_03_MSTR {parent=lowerarm_twist_02_l, loc=<-2.9951 1.213 4.4485>, rot=<0.0962 0.0217 0.7144 -0.6927>}
 dyn_armTie_03 {parent=dyn_armTie_03_MSTR, loc=<0 0 0>, rot=<0 0 0 1>}
 lowerarm_twist_01_l {parent=lowerarm_l, loc=<17.8549 0 0>, rot=<0 0 0 -1>}
 deform_elbow_in_base_l {parent=lowerarm_l, loc=<0.1385 3.6052 -0.2182>, rot=<0.0059 0.0117 -0.0071 -0.9999>}
 deform_elbow_in_l {parent=deform_elbow_in_base_l, loc=<0 0 0>, rot=<0 0 0 1>}
 deform_elbow_out_base_l {parent=lowerarm_l, loc=<0.0565 -3.9993 0.1259>, rot=<0.0059 0.0117 -0.0071 -0.9999>}
 deform_elbow_out_l {parent=deform_elbow_out_base_l, loc=<0 0 0>, rot=<0 0 0 1>}
 upperarm_twist_01_l {parent=upperarm_l, loc=<0.0041 0 0>, rot=<0 0 0 1>}
 deform_upperarm_bk_base_l {parent=upperarm_twist_01_l, loc=<0.9473 -4.46 0.3113>, rot=<0.0004 -0.0034 -0.0158 -0.9999>}
 deform_upperarm_bk_l {parent=deform_upperarm_bk_base_l, loc=<0 0 0>, rot=<0 0 0 1>}
 deform_upperarm_fr_base_l {parent=upperarm_twist_01_l, loc=<1.4228 5.0263 0.3212>, rot=<0.0004 -0.0034 -0.0158 -0.9999>}
 deform_upperarm_fr_l {parent=deform_upperarm_fr_base_l, loc=<0 0 0>, rot=<0 0 0 1>}
 deform_upperarm_up_base_l {parent=upperarm_twist_01_l, loc=<-1.6653 0.3195 3.9629>, rot=<0.0004 -0.0034 -0.0158 -0.9999>}
 deform_upperarm_up_l {parent=deform_upperarm_up_base_l, loc=<0 0 0>, rot=<0 0 0 1>}
 dyn_armTie_01_MSTR {parent=upperarm_twist_01_l, loc=<9.026 0.0016 2.8548>, rot=<0.1835 0.1597 0.6237 -0.7429>}
 dyn_armTie_01 {parent=dyn_armTie_01_MSTR, loc=<0 0 0>, rot=<0 0 0 1>}
 dyn_armTie_02_MSTR {parent=upperarm_twist_01_l, loc=<15.8252 0.1348 2.9429>, rot=<0.1183 0.1312 0.6164 -0.7673>}
 dyn_armTie_02 {parent=dyn_armTie_02_MSTR, loc=<0 0 0>, rot=<0 0 0 1>}
 upperarm_twist_02_l {parent=upperarm_l, loc=<13.7941 0 0>, rot=<0 0 0 1>}
 clavicle_r {parent=spine_05, loc=<5.4831 -0.1053 3.2946>, rot=<0.7388 0.0938 -0.6666 -0.0307>}
 upperarm_r {parent=clavicle_r, loc=<-9.9486 0 0>, rot=<0.0371 -0.2992 -0.0705 -0.9509>}
 lowerarm_r {parent=upperarm_r, loc=<-26.4793 0 0>, rot=<0 0 -0.2382 -0.9712>}
 hand_r {parent=lowerarm_r, loc=<-22.7458 0 0>, rot=<-0.6747 -0.0182 0.0605 -0.7354>}
 thumb_01_r {parent=hand_r, loc=<-1.3118 0.3191 2.459>, rot=<0.6546 -0.4042 0.0664 -0.6354>}
 thumb_02_r {parent=thumb_01_r, loc=<-5.9322 0.0253 -0.1416>, rot=<0.0121 0.0046 0.2964 -0.955>}
 thumb_03_r {parent=thumb_02_r, loc=<-2.2223 -0.3682 0>, rot=<0 0 0.0172 -0.9999>}
 index_metacarpal_r {parent=hand_r, loc=<-1.6535 -0.0183 2.96>, rot=<0.0158 -0.0124 -0.0173 -0.9996>}
 index_01_r {parent=index_metacarpal_r, loc=<-8.3371 -0.1718 0.2461>, rot=<0.0363 0.0499 0.142 -0.9879>}
 index_02_r {parent=index_01_r, loc=<-3.4919 0.3004 0>, rot=<0 0 0.2948 -0.9556>}
 index_03_r {parent=index_02_r, loc=<-2.5578 0.123 0>, rot=<0 0 0.1968 -0.9804>}
 middle_metacarpal_r {parent=hand_r, loc=<-2.1856 0.1725 0.5984>, rot=<0.006 0.0168 -0.0329 -0.9993>}
 middle_01_r {parent=middle_metacarpal_r, loc=<-7.6374 -0.346 0.1508>, rot=<-0.0085 0.0328 0.1813 -0.9829>}
 middle_02_r {parent=middle_01_r, loc=<-3.9108 0 0>, rot=<0 0 0.2808 -0.9598>}
 middle_03_r {parent=middle_02_r, loc=<-2.6103 -0.0517 0>, rot=<0 0 0.1446 -0.9895>}
 ring_metacarpal_r {parent=hand_r, loc=<-2.2074 0.3625 -1.3638>, rot=<-0.0284 0.0722 -0.0458 -0.9959>}
 ring_01_r {parent=ring_metacarpal_r, loc=<-7.2483 -0.0329 0.1426>, rot=<-0.0456 -0.0277 0.2951 -0.954>}
 ring_02_r {parent=ring_01_r, loc=<-3.7208 -0.4647 0>, rot=<0 0 0.1583 -0.9874>}
 ring_03_r {parent=ring_02_r, loc=<-2.4152 -0.114 0>, rot=<0 0 0.1504 -0.9886>}
 pinky_metacarpal_r {parent=hand_r, loc=<-1.7616 0.9753 -3.3814>, rot=<-0.0397 0.0814 -0.069 -0.9935>}
 pinky_01_r {parent=pinky_metacarpal_r, loc=<-7.1933 0.4205 0.2601>, rot=<-0.0487 -0.0063 0.2556 -0.9655>}
 pinky_02_r {parent=pinky_01_r, loc=<-2.8274 -0.0893 0>, rot=<0 0 0.2474 -0.9689>}
 pinky_03_r {parent=pinky_02_r, loc=<-2.2046 -0.0307 0>, rot=<0 0 0.125 -0.9922>}
 deform_wrist_dn_base_r {parent=hand_r, loc=<-0.0001 1.5328 0>, rot=<0 0 0 -1>}
 deform_wrist_dn_r {parent=deform_wrist_dn_base_r, loc=<0 0 0>, rot=<0 0 0 1>}
 deform_wrist_up_base_r {parent=hand_r, loc=<-0.0002 -2.3007 -0.0001>, rot=<0 0 0 -1>}
 deform_wrist_up_r {parent=deform_wrist_up_base_r, loc=<0 0 0>, rot=<0 0 0 1>}
 weapon_r {parent=hand_r, loc=<0.1974 3.4111 -0.3811>, rot=<0 0 0 1>}
 hand_attach_r {parent=hand_r, loc=<-8.5001 2.9999 0>, rot=<0 0 -0.7071 -0.7071>}
 lowerarm_twist_02_r {parent=lowerarm_r, loc=<-9.6199 0 0>, rot=<0 0 0 -1>}
 lowerarm_twist_01_r {parent=lowerarm_r, loc=<-17.8549 0 0>, rot=<0 0 0 -1>}
 deform_elbow_in_base_r {parent=lowerarm_r, loc=<-0.1385 -3.6052 0.2182>, rot=<0.0059 0.0117 -0.0071 -0.9999>}
 deform_elbow_in_r {parent=deform_elbow_in_base_r, loc=<0 0 0>, rot=<0 0 0 1>}
 deform_elbow_out_base_r {parent=lowerarm_r, loc=<-0.0565 3.9993 -0.126>, rot=<0.0059 0.0117 -0.0071 -0.9999>}
 deform_elbow_out_r {parent=deform_elbow_out_base_r, loc=<0 0 0>, rot=<0 0 0 1>}
 upperarm_twist_01_r {parent=upperarm_r, loc=<-0.0041 0 0>, rot=<0 0 0 -1>}
 deform_upperarm_bk_base_r {parent=upperarm_twist_01_r, loc=<-0.9473 4.46 -0.3113>, rot=<0.0004 -0.0034 -0.0158 -0.9999>}
 deform_upperarm_bk_r {parent=deform_upperarm_bk_base_r, loc=<0 0 0>, rot=<0 0 0 1>}
 deform_upperarm_fr_base_r {parent=upperarm_twist_01_r, loc=<-1.4228 -5.0263 -0.3212>, rot=<0.0004 -0.0034 -0.0158 -0.9999>}
 deform_upperarm_fr_r {parent=deform_upperarm_fr_base_r, loc=<0 0 0>, rot=<0 0 0 1>}
 deform_upperarm_up_base_r {parent=upperarm_twist_01_r, loc=<1.6653 -0.3195 -3.9628>, rot=<0.0004 -0.0034 -0.0158 -0.9999>}
 deform_upperarm_up_r {parent=deform_upperarm_up_base_r, loc=<0 0 0>, rot=<0 0 0 1>}
 upperarm_twist_02_r {parent=upperarm_r, loc=<-13.7941 0 0>, rot=<0 0 0 -1>}
 dyn_necklace_MSTR {parent=spine_05, loc=<7.8996 5.4797 -0.0004>, rot=<0.2223 0.6712 0.6712 0.2223>}
 dyn_necklace {parent=dyn_necklace_MSTR, loc=<0 0 0>, rot=<0 0 0 1>}
 dyn_collar {parent=spine_05, loc=<0 0 0>, rot=<0 0 0 1>}
 deform_pec_aim_base_l {parent=spine_04, loc=<3.1942 9.3086 -1.2237>, rot=<0.0614 -0.4031 0.3295 -0.8516>}
 deform_pec_aim_l {parent=deform_pec_aim_base_l, loc=<0 0 0>, rot=<0 0 0 1>}
 deform_pec_base_l {parent=deform_pec_aim_l, loc=<9.3665 0 0>, rot=<0 0 0 1>}
 deform_pec_l {parent=deform_pec_base_l, loc=<0 0 0>, rot=<0 0 0 1>}
 deform_pec_aim_base_r {parent=spine_04, loc=<3.1938 9.3085 1.2237>, rot=<-0.4031 -0.0614 0.8516 0.3295>}
 deform_pec_aim_r {parent=deform_pec_aim_base_r, loc=<0 0 0>, rot=<0 0 0 1>}
 deform_pec_base_r {parent=deform_pec_aim_r, loc=<-9.3667 -0.0003 -0.0004>, rot=<0 0 0 1>}
 deform_pec_r {parent=deform_pec_base_r, loc=<0 0 0>, rot=<0 0 0 1>}
 thigh_l {parent=pelvis, loc=<-1.0855 -0.2109 -8.7053>, rot=<0.0009 0.0345 0.0155 -0.9993>}
 calf_l {parent=thigh_l, loc=<-42.5268 0 0>, rot=<0 0 -0.0657 -0.9978>}
 foot_l {parent=calf_l, loc=<-46.8037 0 0>, rot=<0.0422 -0.04 0.0613 -0.9964>}
 ball_l {parent=foot_l, loc=<-3.5658 10.8066 -0.1544>, rot=<0.0001 -0.0001 -0.7063 -0.7079>}
 calf_twist_02_l {parent=calf_l, loc=<-18.8064 -0.0967 0.0397>, rot=<0.0237 0.0095 -0.0035 -0.9997>}
 dyn_legL_bow_BASE {parent=calf_twist_02_l, loc=<-9.7211 -4.1207 -2.9505>, rot=<-0.0683 -0.7365 0.0319 -0.6722>}
 dyn_legL_bowL_MSTR {parent=dyn_legL_bow_BASE, loc=<0 0 0>, rot=<-0.704 -0.0665 -0.0665 -0.704>}
 dyn_legL_bowL {parent=dyn_legL_bowL_MSTR, loc=<0 0 0>, rot=<0 0 0 1>}
 dyn_legL_bowR_MSTR {parent=dyn_legL_bow_BASE, loc=<0 0 0>, rot=<0.2759 -0.6511 0.6511 -0.2759>}
 dyn_legL_bowR {parent=dyn_legL_bowR_MSTR, loc=<0 0 0>, rot=<0 0 0 1>}
 calf_twist_01_l {parent=calf_l, loc=<-35.813 -0.0967 0.0397>, rot=<0.0237 0.0095 -0.0035 -0.9997>}
 dyn_leg_00_ft_l_MSTR {parent=calf_l, loc=<-8.6826 3.9261 -2.7749>, rot=<-0.0725 -0.8646 -0.4957 -0.0386>}
 dyn_leg_01_ft_l {parent=dyn_leg_00_ft_l_MSTR, loc=<0 0 0>, rot=<0 0 0 1>}
 dyn_leg_02_ft_l {parent=dyn_leg_01_ft_l, loc=<11 0 0>, rot=<0 0 0 1>}
 dyn_leg_03_ft_l {parent=dyn_leg_02_ft_l, loc=<11 0 0>, rot=<0 0 0 1>}
 dyn_leg_00_in_l_MSTR {parent=calf_l, loc=<-8.6826 -1.397 5.6818>, rot=<0.0436 -0.0015 0.9984 0.0349>}
 dyn_leg_01_in_l {parent=dyn_leg_00_in_l_MSTR, loc=<0 0 0>, rot=<0 0 0 1>}
 dyn_leg_02_in_l {parent=dyn_leg_01_in_l, loc=<11 0 0>, rot=<0 0 0 1>}
 dyn_leg_03_in_l {parent=dyn_leg_02_in_l, loc=<11 0 0>, rot=<0 0 0 1>}
 dyn_leg_00_bk_l_MSTR {parent=calf_l, loc=<-8.6826 -7.2877 -3.1027>, rot=<0.0254 -0.8649 0.4994 -0.044>}
 dyn_leg_01_bk_l {parent=dyn_leg_00_bk_l_MSTR, loc=<0 0 0>, rot=<0 0 0 1>}
 dyn_leg_02_bk_l {parent=dyn_leg_01_bk_l, loc=<11 0 0>, rot=<0 0 0 1>}
 dyn_leg_03_bk_l {parent=dyn_leg_02_bk_l, loc=<11 0 0>, rot=<0 0 0 1>}
 deform_knee_in_base_l {parent=calf_l, loc=<0.5405 -5.0592 -0.5944>, rot=<-0.0003 -0.0151 0.0267 -0.9995>}
 deform_knee_in_l {parent=deform_knee_in_base_l, loc=<0 0 0>, rot=<0 0 0 1>}
 deform_knee_out_base_l {parent=calf_l, loc=<0.2138 3.9943 -0.8167>, rot=<-0.0003 -0.0151 0.0267 -0.9995>}
 deform_knee_out_l {parent=deform_knee_out_base_l, loc=<0 0 0>, rot=<0 0 0 1>}
 dyn_leg_00_inbk_l_MSTR {parent=calf_l, loc=<-8.6826 -6.1034 3.0156>, rot=<0.0365 -0.4103 0.9112 0.0059>}
 dyn_leg_01_inbk_l {parent=dyn_leg_00_inbk_l_MSTR, loc=<0 0 0>, rot=<0 0 0 1>}
 dyn_leg_02_inbk_l {parent=dyn_leg_01_inbk_l, loc=<11 0 0>, rot=<0 0 0 1>}
 dyn_leg_03_inbk_l {parent=dyn_leg_02_inbk_l, loc=<11 0 0>, rot=<0 0 0 1>}
 thigh_twist_01_l {parent=thigh_l, loc=<-21.629 -0.0919 -0.0002>, rot=<0.0013 0 0 -1>}
 dyn_legStrap_l_MSTR {parent=thigh_twist_01_l, loc=<14.9918 -0.4585 -7.6756>, rot=<0.0441 -0.9959 0.0024 -0.0789>}
 dyn_legStrap_l {parent=dyn_legStrap_l_MSTR, loc=<0 0 0>, rot=<0 0 0 1>}
 deform_glute_base_l {parent=thigh_l, loc=<-6.0004 -8.9997 -0.0128>, rot=<0.0004 0.0009 -0.4617 -0.887>}
 deform_glute_l {parent=deform_glute_base_l, loc=<-0.0006 -0.0014 0>, rot=<0 0 0 1>}
 deform_groin_base_l {parent=thigh_l, loc=<0.4808 5.6776 -0.2472>, rot=<0 0.001 0 -1>}
 deform_groin_l {parent=deform_groin_base_l, loc=<0 0 0>, rot=<0 0 0 1>}
 thigh_r {parent=pelvis, loc=<-1.0855 -0.2109 8.7053>, rot=<0.0345 -0.0009 0.9993 0.0155>}
 calf_r {parent=thigh_r, loc=<42.5268 0 0>, rot=<0 0 -0.0657 -0.9978>}
 foot_r {parent=calf_r, loc=<46.8037 0 0>, rot=<0.0422 -0.04 0.0613 -0.9964>}
 ball_r {parent=foot_r, loc=<3.5658 -10.8066 0.1544>, rot=<0.0001 -0.0001 -0.7063 -0.7079>}
 calf_twist_02_r {parent=calf_r, loc=<18.8064 0.0967 -0.0397>, rot=<0.0237 0.0095 -0.0035 -0.9997>}
 dyn_legR_bow_BASE {parent=calf_twist_02_r, loc=<9.7211 4.1207 2.9504>, rot=<-0.6722 0.0319 0.7365 0.0683>}
 dyn_legR_bowR_MSTR {parent=dyn_legR_bow_BASE, loc=<0 0 0>, rot=<-0.0665 -0.704 0.704 0.0665>}
 dyn_legR_bowR {parent=dyn_legR_bowR_MSTR, loc=<0 0 0>, rot=<0 0 0 1>}
 dyn_legR_bowL_MSTR {parent=dyn_legR_bow_BASE, loc=<0 0 0>, rot=<-0.6511 0.2759 0.2759 -0.6511>}
 dyn_legR_bowL {parent=dyn_legR_bowL_MSTR, loc=<0 0 0>, rot=<0 0 0 1>}
 calf_twist_01_r {parent=calf_r, loc=<35.813 0.0967 -0.0397>, rot=<0.0237 0.0095 -0.0035 -0.9997>}
 dyn_leg_00_ft_r_MSTR {parent=calf_r, loc=<8.6825 -3.9261 2.7749>, rot=<-0.4957 0.0386 0.0725 -0.8646>}
 dyn_leg_01_ft_r {parent=dyn_leg_00_ft_r_MSTR, loc=<0 0 0>, rot=<0 0 0 1>}
 dyn_leg_02_ft_r {parent=dyn_leg_01_ft_r, loc=<11.0001 0 0>, rot=<0 0 0 1>}
 dyn_leg_03_ft_r {parent=dyn_leg_02_ft_r, loc=<11 0 0>, rot=<0 0 0 1>}
 dyn_leg_00_in_r_MSTR {parent=calf_r, loc=<8.6826 1.397 -5.6818>, rot=<0.9984 -0.0349 -0.0436 -0.0015>}
 dyn_leg_01_in_r {parent=dyn_leg_00_in_r_MSTR, loc=<0 0 0>, rot=<0 0 0 1>}
 dyn_leg_02_in_r {parent=dyn_leg_01_in_r, loc=<11 0 0>, rot=<0 0 0 1>}
 dyn_leg_03_in_r {parent=dyn_leg_02_in_r, loc=<11 0 0>, rot=<0 0 0 1>}
 dyn_leg_00_bk_r_MSTR {parent=calf_r, loc=<8.6826 7.2877 3.1026>, rot=<0.4994 0.044 -0.0254 -0.8649>}
 dyn_leg_01_bk_r {parent=dyn_leg_00_bk_r_MSTR, loc=<0 0 0>, rot=<0 0 0 1>}
 dyn_leg_02_bk_r {parent=dyn_leg_01_bk_r, loc=<11 0 0>, rot=<0 0 0 -1>}
 dyn_leg_03_bk_r {parent=dyn_leg_02_bk_r, loc=<11 0.0001 0>, rot=<0 0 0 -1>}
 deform_knee_in_base_r {parent=calf_r, loc=<-0.5405 5.0592 0.5943>, rot=<-0.0003 -0.0151 0.0267 -0.9995>}
 deform_knee_in_r {parent=deform_knee_in_base_r, loc=<0 0 0>, rot=<0 0 0 1>}
 deform_knee_out_base_r {parent=calf_r, loc=<-0.2137 -3.9943 0.8168>, rot=<-0.0003 -0.0151 0.0267 -0.9995>}
 deform_knee_out_r {parent=deform_knee_out_base_r, loc=<0 0 0>, rot=<0 0 0 1>}
 dyn_leg_00_inbk_r_MSTR {parent=calf_r, loc=<8.6826 6.1034 -3.0156>, rot=<0.9112 -0.0059 -0.0365 -0.4103>}
 dyn_leg_01_inbk_r {parent=dyn_leg_00_inbk_r_MSTR, loc=<0 0 0>, rot=<0 0 0 1>}
 dyn_leg_02_inbk_r {parent=dyn_leg_01_inbk_r, loc=<11 0 0>, rot=<0 0 0 -1>}
 dyn_leg_03_inbk_r {parent=dyn_leg_02_inbk_r, loc=<11 0 0>, rot=<0 0 0 -1>}
 thigh_twist_01_r {parent=thigh_r, loc=<21.629 0.0919 0.0002>, rot=<0.0013 0 0 -1>}
 dyn_legStrap_r_MSTR {parent=thigh_twist_01_r, loc=<-15.0278 0.3769 7.5759>, rot=<-0.0061 0.182 0.0265 -0.9829>}
 dyn_legStrap_r {parent=dyn_legStrap_r_MSTR, loc=<0 0 0>, rot=<0 0 0 1>}
 deform_glute_base_r {parent=thigh_r, loc=<6.0002 8.9997 0.0129>, rot=<0.0004 0.0009 -0.4617 -0.887>}
 deform_glute_r {parent=deform_glute_base_r, loc=<0.0006 0.0014 0>, rot=<0 0 0 1>}
 deform_groin_base_r {parent=thigh_r, loc=<-0.4808 -5.6776 0.2472>, rot=<0 0.001 0 -1>}
 deform_groin_r {parent=deform_groin_base_r, loc=<0 0 0>, rot=<0 0 0 1>}
 dyn_belt_MSTR {parent=pelvis, loc=<3.1784 -1.329 0>, rot=<-0.0127 -0.707 0.0127 -0.707>}
 dyn_belt_l_pivot {parent=dyn_belt_MSTR, loc=<0 0 0>, rot=<0 0 0 1>}
 dyn_belt_l_MSTR {parent=dyn_belt_l_pivot, loc=<15.0406 0 -1.1033>, rot=<0 0 0 1>}
 dyn_belt_l {parent=dyn_belt_l_MSTR, loc=<0 0 0>, rot=<0 0 0 1>}
 dyn_belt_02_ft_l_pivot {parent=dyn_belt_MSTR, loc=<0 0 0>, rot=<0 0 0 1>}
 dyn_belt_02_ft_l_MSTR {parent=dyn_belt_02_ft_l_pivot, loc=<11.9781 6.9156 -1.7498>, rot=<0 0 -0.3746 -0.9272>}
 dyn_belt_02_ft_l {parent=dyn_belt_02_ft_l_MSTR, loc=<0 0 0>, rot=<0 0 0 1>}
 dyn_pouch_MOVE {parent=dyn_belt_02_ft_l_MSTR, loc=<0 0 0>, rot=<0 0 0 1>}
 dyn_pouch_MSTR {parent=dyn_pouch_MOVE, loc=<0.2569 0.3725 -1.5001>, rot=<0.0524 -0.6164 -0.0666 -0.7829>}
 dyn_pouch {parent=dyn_pouch_MSTR, loc=<0 0 0>, rot=<0 0 0 1>}
 dyn_pouchStrap_l {parent=dyn_pouch_MOVE, loc=<-0.8984 -1.8906 -0.1127>, rot=<0.0524 -0.6164 -0.0666 -0.7829>}
 dyn_pouchStrap_r {parent=dyn_pouch_MOVE, loc=<-1.8442 2.5097 -0.0152>, rot=<0.0524 -0.6164 -0.0666 -0.7829>}
 dyn_belt_01_ft_l_pivot {parent=dyn_belt_MSTR, loc=<0 0 0>, rot=<0 0 0 1>}
 dyn_belt_01_ft_l_MSTR {parent=dyn_belt_01_ft_l_pivot, loc=<5.5144 9.5513 -1.1959>, rot=<0 0 -0.6018 -0.7986>}
 dyn_belt_01_ft_l {parent=dyn_belt_01_ft_l_MSTR, loc=<0 0 0>, rot=<0 0 0 1>}
 dyn_belt_ft_pivot {parent=dyn_belt_MSTR, loc=<0 0 0>, rot=<0 0 0 1>}
 dyn_belt_ft_MSTR {parent=dyn_belt_ft_pivot, loc=<0 10.1443 -0.4326>, rot=<0 0 -0.7071 -0.7071>}
 dyn_belt_ft {parent=dyn_belt_ft_MSTR, loc=<0 0 0>, rot=<0 0 0 1>}
 dyn_belt_01_ft_r_pivot {parent=dyn_belt_MSTR, loc=<0 0 0>, rot=<0 0 0 1>}
 dyn_belt_01_ft_r_MSTR {parent=dyn_belt_01_ft_r_pivot, loc=<-5.4858 9.5016 0.5869>, rot=<0 0 -0.809 -0.5878>}
 dyn_belt_01_ft_r {parent=dyn_belt_01_ft_r_MSTR, loc=<0 0 0>, rot=<0 0 0 1>}
 dyn_belt_02_ft_r_pivot {parent=dyn_belt_MSTR, loc=<0 0 0>, rot=<0 0 0 1>}
 dyn_belt_02_ft_r_MSTR {parent=dyn_belt_02_ft_r_pivot, loc=<-11.1634 6.4452 1.7889>, rot=<0 0 0.9304 0.3665>}
 dyn_belt_02_ft_r {parent=dyn_belt_02_ft_r_MSTR, loc=<0 0 0>, rot=<0 0 0 1>}
 dyn_belt_r_pivot {parent=dyn_belt_MSTR, loc=<0 0 0>, rot=<0 0 0 1>}
 dyn_belt_r_MSTR {parent=dyn_belt_r_pivot, loc=<-13.5695 0 3.4539>, rot=<0 0 1 0>}
 dyn_belt_r {parent=dyn_belt_r_MSTR, loc=<0 0 0>, rot=<0 0 0 1>}
 dyn_belt_bk_r_pivot {parent=dyn_belt_MSTR, loc=<0 0 0>, rot=<0 0 0 1>}
 dyn_belt_bk_r_MSTR {parent=dyn_belt_bk_r_pivot, loc=<-7.9594 -7.9594 3.9116>, rot=<0 0 0.887 -0.4617>}
 dyn_belt_bk_r {parent=dyn_belt_bk_r_MSTR, loc=<0 0 0>, rot=<0 0 0 1>}
 dyn_belt_bk_pivot {parent=dyn_belt_MSTR, loc=<0 0 0>, rot=<0 0 0 1>}
 dyn_belt_bk_MSTR {parent=dyn_belt_bk_pivot, loc=<0 -10.1012 3.0795>, rot=<0 0 0.7071 -0.7071>}
 dyn_belt_bk {parent=dyn_belt_bk_MSTR, loc=<0 0 0>, rot=<0 0 0 1>}
 dyn_belt_bk_l_pivot {parent=dyn_belt_MSTR, loc=<0 0 0>, rot=<0 0 0 1>}
 dyn_belt_bk_l_MSTR {parent=dyn_belt_bk_l_pivot, loc=<8.8805 -8.8805 1.6365>, rot=<0 0 0.4617 -0.887>}
 dyn_belt_bk_l {parent=dyn_belt_bk_l_MSTR, loc=<0 0 0>, rot=<0 0 0 1>}
 dyn_charm_MSTR {parent=pelvis, loc=<8.3848 8.7176 2.9468>, rot=<-0.0084 -0.0394 0.9771 0.2088>}
 dyn_charm {parent=dyn_charm_MSTR, loc=<0 0 0>, rot=<0 0 0 1>}
 dyn_chain_00_MSTR {parent=pelvis, loc=<10.4102 6.0573 6.6831>, rot=<0.0078 -0.0805 0.9695 0.2311>}
 dyn_chain_01 {parent=dyn_chain_00_MSTR, loc=<0 0 0>, rot=<0 0 0 1>}
 dyn_chain_02 {parent=dyn_chain_01, loc=<5.5 0 0>, rot=<0 0 -0.1895 -0.9819>}
 dyn_chain_03 {parent=dyn_chain_02, loc=<5.5 0 0>, rot=<0.0079 0.0028 -0.029 -0.9995>}
 dyn_chain_04 {parent=dyn_chain_03, loc=<5.5 0 0>, rot=<0 0 0 1>}
 dyn_chain_05 {parent=dyn_chain_04, loc=<5.5 0 0>, rot=<0 0 0 1>}
 dyn_hipStraps_MSTR {parent=pelvis, loc=<5.1677 -1.5853 12.7347>, rot=<0.1736 -0.0031 0.9846 0.0176>}
 dyn_hipStraps {parent=dyn_hipStraps_MSTR, loc=<0 0 0>, rot=<0 0 0 1>}
 dyn_hipLoop_r_MSTR {parent=pelvis, loc=<10.0372 -1.2577 10.6856>, rot=<0.3826 -0.0069 0.9237 0.0166>}
 dyn_hipLoop_r {parent=dyn_hipLoop_r_MSTR, loc=<0 0 0>, rot=<0 0 0 1>}
 dyn_hipLoop_l_MSTR {parent=pelvis, loc=<10.0539 -1.1594 -10.7255>, rot=<-0.0166 -0.9237 0.0069 -0.3826>}
 dyn_hipLoop_l {parent=dyn_hipLoop_l_MSTR, loc=<0 0 0>, rot=<0 0 0 1>}
 attach {parent=root, loc=<0 0 0>, rot=<0 0 0 1>}
 ik_foot_root {parent=root, loc=<0 0 0>, rot=<0 0 0 1>}
 ik_foot_l {parent=ik_foot_root, loc=<14.8262 -7.3545 8.296>, rot=<0.0357 0.7045 0.0284 -0.7082>}
 ik_foot_r {parent=ik_foot_root, loc=<-14.8262 -7.3545 8.296>, rot=<0.7082 0.0284 -0.7045 0.0357>}
 ik_hand_root {parent=root, loc=<0 0 0>, rot=<0 0 0 1>}
 ik_hand_gun {parent=ik_hand_root, loc=<-48.1743 3.3545 104.4381>, rot=<0.6276 0.1309 0.4183 -0.6434>}
 ik_hand_r {parent=ik_hand_gun, loc=<0 0 0>, rot=<0 0 0 1>}
 ik_hand_l {parent=ik_hand_gun, loc=<59.3271 67.6997 34.3524>, rot=<-0.6158 -0.7027 -0.3565 0>}
]
variables [isCrouching {type=bool, default=False}]
