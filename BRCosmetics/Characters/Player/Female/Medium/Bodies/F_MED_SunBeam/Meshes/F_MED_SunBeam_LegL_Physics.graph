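# Physics asset for the left-leg mesh of the SunBeam female medium body, with four dynamic bone chains hanging from the calf.
# Front, inner, back and inner-back chains each run three capsules from a tiny master body, with angular limits widening down the chain and tighter twist on inner chains.
# Free-angular cross constraints tie the lower inner-back segments to the back ones.
# Calf and foot capsules plus a flat root box supply body and ground shapes.
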PhysicsAsset /BRCosmetics/Characters/Player/Female/Medium/Bodies/F_MED_SunBeam/Meshes/F_MED_SunBeam_LegL_Physics
bodies {
 calf_l [{shapeType=Capsule, length=47, radius=10, shapeName=calf_l_capsule, center=<-23 0 0>, rotation=<rot 90 0 0>}]
 dyn_leg_00_bk_l_MSTR [{shapeType=Capsule, length=0.505, radius=0.505, shapeName=dyn_leg_00_bk_l_MSTR_capsule}]
 dyn_leg_00_ft_l_MSTR [{shapeType=Capsule, length=0.505, radius=0.505, shapeName=dyn_leg_00_ft_l_MSTR_capsule}]
 dyn_leg_00_in_l_MSTR [{shapeType=Capsule, length=0.505, radius=0.505, shapeName=dyn_leg_00_in_l_MSTR_capsule, center=<5 0 0>}]
 dyn_leg_00_inbk_l_MSTR [{shapeType=Capsule, length=0.505, radius=0.505, shapeName=dyn_leg_00_inbk_l_MSTR_capsule}]
 dyn_leg_01_bk_l [{shapeType=Capsule, length=17.5, radius=10, shapeName=dyn_leg_01_bk_l_capsule, center=<4 0 12>, rotation=<rot 0 0 90>}]
 dyn_leg_01_ft_l [{shapeType=Capsule, length=17.5, radius=10, shapeName=dyn_leg_01_ft_l_capsule, center=<4 0 15>, rotation=<rot 0 0 90>}]
 dyn_leg_01_in_l [{shapeType=Capsule, length=17.5, radius=10, shapeName=dyn_leg_01_in_l_capsule, center=<4 -0.0038 14.5>, rotation=<rot 0 0 90>}]
 dyn_leg_01_inbk_l [{shapeType=Capsule, length=17.5, radius=10, shapeName=dyn_leg_01_inbk_l_capsule, center=<5.5 0 13.5>, rotation=<rot 0 0 90>}]
 dyn_leg_02_bk_l [{shapeType=Capsule, length=17.5, radius=10, shapeName=dyn_leg_02_bk_l_capsule, center=<4 0 11>, rotation=<rot 0 0 90>}]
 dyn_leg_02_ft_l [{shapeType=Capsule, length=17.5, radius=10, shapeName=dyn_leg_02_ft_l_capsule, center=<4 0 13.5>, rotation=<rot 0 0 90>}]
 dyn_leg_02_in_l [{shapeType=Capsule, length=17.5, radius=10, shapeName=dyn_leg_02_in_l_capsule, center=<4 -0.0038 13.5>, rotation=<rot 0 0 90>}]
 dyn_leg_02_inbk_l [{shapeType=Capsule, length=17.5, radius=10, shapeName=dyn_leg_02_inbk_l_capsule, center=<5.5 0 13>, rotation=<rot 0 0 90>}]
 dyn_leg_03_bk_l [{shapeType=Capsule, length=17.5, radius=10, shapeName=dyn_leg_03_bk_l_capsule, center=<4 0 10>, rotation=<rot 0 0 90>}]
 dyn_leg_03_ft_l [{shapeType=Capsule, length=17.5, radius=10, shapeName=dyn_leg_03_ft_l_capsule, center=<4 0 12>, rotation=<rot 0 0 90>}]
 dyn_leg_03_in_l [{shapeType=Capsule, length=17.5, radius=10, shapeName=dyn_leg_03_in_l_capsule, center=<4 -0.0038 12.5>, rotation=<rot 0 0 90>}]
 dyn_leg_03_inbk_l [{shapeType=Capsule, length=17.5, radius=10, shapeName=dyn_leg_03_inbk_l_capsule, center=<5.5 0 12>, rotation=<rot 0 0 90>}]
 foot_l [{shapeType=Capsule, length=8.3381, radius=7.8314, shapeName=foot_l_capsule, center=<-4.013 4.0957 -0.0895>, rotation=<rot 0 9.6588 -88.2973>}]
 pelvis [{shapeType=Capsule, length=0.505, radius=0.505, shapeName=pelvis_capsule}]
 root [{shapeType=Box, extentX=175, extentY=175, extentZ=10, shapeName=root_box, center=<0 0 -5>}]
 thigh_l [{shapeType=Capsule, length=0.505, radius=0.505, shapeName=thigh_l_capsule}]
}
constraints [
 {
  bone1Name=pelvis
  bone2Name=root
  swing1LimitDegrees=45
  swing1Motion=Free
  swing2LimitDegrees=45
  swing2Motion=Free
  twistLimitDegrees=45
  twistMotion=Free
 }
 {
  bone1Name=thigh_l
  bone2Name=pelvis
  swing1LimitDegrees=45
  swing1Motion=Free
  swing2LimitDegrees=45
  swing2Motion=Free
  twistLimitDegrees=45
  twistMotion=Free
 }
 {
  bone1Name=calf_l
  bone2Name=thigh_l
  swing1LimitDegrees=45
  swing1Motion=Free
  swing2LimitDegrees=45
  swing2Motion=Free
  twistLimitDegrees=45
  twistMotion=Free
 }
 {
  bone1Name=foot_l
  bone2Name=calf_l
  swing1LimitDegrees=45
  swing1Motion=Free
  swing2LimitDegrees=45
  swing2Motion=Free
  twistLimitDegrees=45
  twistMotion=Free
 }
 {
  bone1Name=dyn_leg_00_ft_l_MSTR
  bone2Name=calf_l
  swing1LimitDegrees=45
  swing1Motion=Free
  swing2LimitDegrees=45
  swing2Motion=Free
  twistLimitDegrees=45
  twistMotion=Free
 }
 {bone1Name=dyn_leg_03_ft_l, bone2Name=dyn_leg_02_ft_l, swing1LimitDegrees=20, swing2LimitDegrees=20, twistLimitDegrees=20}
 {bone1Name=dyn_leg_02_ft_l, bone2Name=dyn_leg_01_ft_l, swing1LimitDegrees=15, swing2LimitDegrees=15, twistLimitDegrees=15}
 {bone1Name=dyn_leg_01_ft_l, bone2Name=dyn_leg_00_ft_l_MSTR, swing1LimitDegrees=10, swing2LimitDegrees=10, twistLimitDegrees=10}
 {
  bone1Name=dyn_leg_00_in_l_MSTR
  bone2Name=calf_l
  swing1LimitDegrees=45
  swing1Motion=Free
  swing2LimitDegrees=45
  swing2Motion=Free
  twistLimitDegrees=45
  twistMotion=Free
 }
 {bone1Name=dyn_leg_01_in_l, bone2Name=dyn_leg_00_in_l_MSTR, swing1LimitDegrees=10, swing2LimitDegrees=10, twistLimitDegrees=2}
 {bone1Name=dyn_leg_02_in_l, bone2Name=dyn_leg_01_in_l, swing1LimitDegrees=15, swing2LimitDegrees=15, twistLimitDegrees=2}
 {bone1Name=dyn_leg_03_in_l, bone2Name=dyn_leg_02_in_l, swing1LimitDegrees=20, swing2LimitDegrees=20, twistLimitDegrees=2}
 {
  bone1Name=dyn_leg_00_bk_l_MSTR
  bone2Name=calf_l
  swing1LimitDegrees=45
  swing1Motion=Free
  swing2LimitDegrees=45
  swing2Motion=Free
  twistLimitDegrees=45
  twistMotion=Free
 }
 {bone1Name=dyn_leg_01_bk_l, bone2Name=dyn_leg_00_bk_l_MSTR, swing1LimitDegrees=10, swing2LimitDegrees=10, twistLimitDegrees=10}
 {bone1Name=dyn_leg_02_bk_l, bone2Name=dyn_leg_01_bk_l, swing1LimitDegrees=15, swing2LimitDegrees=15, twistLimitDegrees=15}
 {bone1Name=dyn_leg_03_bk_l, bone2Name=dyn_leg_02_bk_l, swing1LimitDegrees=20, swing2LimitDegrees=20, twistLimitDegrees=20}
 {
  bone1Name=dyn_leg_00_inbk_l_MSTR
  bone2Name=calf_l
  swing1LimitDegrees=45
  swing1Motion=Free
  swing2LimitDegrees=45
  swing2Motion=Free
  twistLimitDegrees=45
  twistMotion=Free
 }
 {bone1Name=dyn_leg_01_inbk_l, bone2Name=dyn_leg_00_inbk_l_MSTR, swing1LimitDegrees=10, swing2LimitDegrees=10, twistLimitDegrees=2}
 {bone1Name=dyn_leg_02_inbk_l, bone2Name=dyn_leg_01_inbk_l, swing1LimitDegrees=15, swing2LimitDegrees=15, twistLimitDegrees=2}
 {bone1Name=dyn_leg_03_inbk_l, bone2Name=dyn_leg_02_inbk_l, swing1LimitDegrees=20, swing2LimitDegrees=20, twistLimitDegrees=2}
 {
  bone1Name=dyn_leg_02_inbk_l
  bone2Name=dyn_leg_02_bk_l
  swing1LimitDegrees=45
  swing1Motion=Free
  swing2LimitDegrees=45
  swing2Motion=Free
  twistLimitDegrees=45
  twistMotion=Free
 }
 {
  bone1Name=dyn_leg_03_inbk_l
  bone2Name=dyn_leg_03_bk_l
  swing1LimitDegrees=45
  swing1Motion=Free
  swing2LimitDegrees=45
  swing2Motion=Free
  twistLimitDegrees=45
  twistMotion=Free
 }
]
