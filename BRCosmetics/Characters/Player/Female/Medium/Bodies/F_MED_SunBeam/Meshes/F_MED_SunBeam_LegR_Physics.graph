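# Physics asset for the right leg piece of the SunBeam female medium body, providing bodies for its dynamic leg bone chains.
# Four three-segment capsule chains (front, inner, back, inner-back) hang from tiny master bodies on the calf, with angular limits loosening toward each tip.
# The pelvis-to-foot links and master joints rotate freely, as does a cross-link between the back and inner-back tips.
# Calf and foot capsules accompany a wide, flat box on the root.
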PhysicsAsset /BRCosmetics/Characters/Player/Female/Medium/Bodies/F_MED_SunBeam/Meshes/F_MED_SunBeam_LegR_Physics
bodies {
 calf_r [{shapeType=Capsule, length=47, radius=10, shapeName=calf_l_capsule, center=<23 0 0>, rotation=<rot 90 81.8176 81.8176>}]
 dyn_leg_00_bk_r_MSTR [{shapeType=Capsule, length=0.505, radius=0.505, shapeName=dyn_leg_00_ft_r_MSTR_capsule, rotation=<rot 0 -90 0>}]
 dyn_leg_00_ft_r_MSTR [{shapeType=Capsule, length=0.505, radius=0.505, shapeName=dyn_leg_00_ft_r_MSTR_capsule, rotation=<rot 0 -90 0>}]
 dyn_leg_00_in_r_MSTR [{shapeType=Capsule, length=0.505, radius=0.505, shapeName=dyn_leg_00_ft_r_MSTR_capsule, rotation=<rot 0 -90 0>}]
 dyn_leg_00_inbk_r_MSTR [{shapeType=Capsule, length=0.505, radius=0.505, shapeName=dyn_leg_00_ft_r_MSTR_capsule, rotation=<rot 0 -90 0>}]
 dyn_leg_01_bk_r [{shapeType=Capsule, length=17.5, radius=10, shapeName=dyn_leg_01_bk_l_capsule, center=<4 0 12>, rotation=<rot 0 0 90>}]
 dyn_leg_01_ft_r [{shapeType=Capsule, length=17.5, radius=10, shapeName=dyn_leg_01_bk_l_capsule, center=<4 0 15>, rotation=<rot 0 0 90>}]
 dyn_leg_01_in_r [{shapeType=Capsule, length=17.5, radius=10, shapeName=dyn_leg_01_bk_l_capsule, center=<4 0 14>, rotation=<rot 0 0 90>}]
 dyn_leg_01_inbk_r [{shapeType=Capsule, length=17.5, radius=10, shapeName=dyn_leg_01_bk_l_capsule, center=<4 0 13.5>, rotation=<rot 0 0 90>}]
 dyn_leg_02_bk_r [{shapeType=Capsule, length=17.5, radius=10, shapeName=dyn_leg_01_bk_l_capsule, center=<4 0 11>, rotation=<rot 0 0 90>}]
 dyn_leg_02_ft_r [{shapeType=Capsule, length=17.5, radius=10, shapeName=dyn_leg_01_bk_l_capsule, center=<4 0 13.5>, rotation=<rot 0 0 90>}]
 dyn_leg_02_in_r [{shapeType=Capsule, length=17.5, radius=10, shapeName=dyn_leg_01_bk_l_capsule, center=<4 0 13>, rotation=<rot 0 0 90>}]
 dyn_leg_02_inbk_r [{shapeType=Capsule, length=17.5, radius=10, shapeName=dyn_leg_01_bk_l_capsule, center=<4 0 13>, rotation=<rot 0 0 90>}]
 dyn_leg_03_bk_r [{shapeType=Capsule, length=17.5, radius=10, shapeName=dyn_leg_01_bk_l_capsule, center=<4 0 10>, rotation=<rot 0 0 90>}]
 dyn_leg_03_ft_r [{shapeType=Capsule, length=17.5, radius=10, shapeName=dyn_leg_01_bk_l_capsule, center=<4 0 12>, rotation=<rot 0 0 90>}]
 dyn_leg_03_in_r [{shapeType=Capsule, length=17.5, radius=10, shapeName=dyn_leg_01_bk_l_capsule, center=<4 0 12>, rotation=<rot 0 0 90>}]
 dyn_leg_03_inbk_r [{shapeType=Capsule, length=17.5, radius=10, shapeName=dyn_leg_01_bk_l_capsule, center=<4 0 12>, rotation=<rot 0 0 90>}]
 foot_r [{shapeType=Capsule, length=8.0136, radius=7.5353, shapeName=foot_r_capsule, center=<3.8588 -3.8488 0.2413>, rotation=<rot 0 5.8466 -89.8434>}]
 pelvis [{shapeType=Capsule, length=0.505, radius=0.505, shapeName=pelvis_capsule}]
 root [{shapeType=Box, extentX=175, extentY=175, extentZ=10, shapeName=root_box, center=<0 0 -5>}]
 thigh_r [{shapeType=Capsule, length=0.505, radius=0.505, shapeName=thigh_l_capsule}]
}
constraints [
 {
  bone1Name=pelvis
  bone2Name=root
  swing1LimitDegrees=45
  swing1Motion=Free
  swing2LimitDegrees=45
  swing2Motion=Free
  twistLimitDegrees=45
  twistMotion=Free
 }
 {
  bone1Name=thigh_r
  bone2Name=pelvis
  swing1LimitDegrees=45
  swing1Motion=Free
  swing2LimitDegrees=45
  swing2Motion=Free
  twistLimitDegrees=45
  twistMotion=Free
 }
 {
  bone1Name=calf_r
  bone2Name=thigh_r
  swing1LimitDegrees=45
  swing1Motion=Free
  swing2LimitDegrees=45
  swing2Motion=Free
  twistLimitDegrees=45
  twistMotion=Free
 }
 {
  bone1Name=foot_r
  bone2Name=calf_r
  swing1LimitDegrees=45
  swing1Motion=Free
  swing2LimitDegrees=45
  swing2Motion=Free
  twistLimitDegrees=45
  twistMotion=Free
 }
 {bone1Name=dyn_leg_01_ft_r, bone2Name=dyn_leg_00_ft_r_MSTR, swing1LimitDegrees=10, swing2LimitDegrees=10, twistLimitDegrees=10}
 {bone1Name=dyn_leg_02_ft_r, bone2Name=dyn_leg_01_ft_r, swing1LimitDegrees=15, swing2LimitDegrees=15, twistLimitDegrees=15}
 {bone1Name=dyn_leg_03_ft_r, bone2Name=dyn_leg_02_ft_r, swing1LimitDegrees=20, swing2LimitDegrees=20, twistLimitDegrees=20}
 {
  bone1Name=dyn_leg_00_ft_r_MSTR
  bone2Name=calf_r
  swing1LimitDegrees=45
  swing1Motion=Free
  swing2LimitDegrees=45
  swing2Motion=Free
  twistLimitDegrees=45
  twistMotion=Free
 }
 {
  bone1Name=dyn_leg_00_in_r_MSTR
  bone2Name=calf_r
  swing1LimitDegrees=45
  swing1Motion=Free
  swing2LimitDegrees=45
  swing2Motion=Free
  twistLimitDegrees=45
  twistMotion=Free
 }
 {bone1Name=dyn_leg_01_in_r, bone2Name=dyn_leg_00_in_r_MSTR, swing1LimitDegrees=10, swing2LimitDegrees=10, twistLimitDegrees=10}
 {bone1Name=dyn_leg_02_in_r, bone2Name=dyn_leg_01_in_r, swing1LimitDegrees=15, swing2LimitDegrees=15, twistLimitDegrees=15}
 {bone1Name=dyn_leg_03_in_r, bone2Name=dyn_leg_02_in_r, swing1LimitDegrees=20, swing2LimitDegrees=20, twistLimitDegrees=20}
 {
  bone1Name=dyn_leg_00_bk_r_MSTR
  bone2Name=calf_r
  swing1LimitDegrees=45
  swing1Motion=Free
  swing2LimitDegrees=45
  swing2Motion=Free
  twistLimitDegrees=45
  twistMotion=Free
 }
 {bone1Name=dyn_leg_01_bk_r, bone2Name=dyn_leg_00_bk_r_MSTR, swing1LimitDegrees=10, swing2LimitDegrees=10, twistLimitDegrees=10}
 {bone1Name=dyn_leg_02_bk_r, bone2Name=dyn_leg_01_bk_r, swing1LimitDegrees=15, swing2LimitDegrees=15, twistLimitDegrees=15}
 {bone1Name=dyn_leg_03_bk_r, bone2Name=dyn_leg_02_bk_r, swing1LimitDegrees=20, swing2LimitDegrees=20, twistLimitDegrees=20}
 {
  bone1Name=dyn_leg_00_inbk_r_MSTR
  bone2Name=calf_r
  swing1LimitDegrees=45
  swing1Motion=Free
  swing2LimitDegrees=45
  swing2Motion=Free
  twistLimitDegrees=45
  twistMotion=Free
 }
 {bone1Name=dyn_leg_01_inbk_r, bone2Name=dyn_leg_00_inbk_r_MSTR, swing1LimitDegrees=10, swing2LimitDegrees=10, twistLimitDegrees=10}
 {bone1Name=dyn_leg_02_inbk_r, bone2Name=dyn_leg_01_inbk_r, swing1LimitDegrees=15, swing2LimitDegrees=15, twistLimitDegrees=15}
 {bone1Name=dyn_leg_03_inbk_r, bone2Name=dyn_leg_02_inbk_r, swing1LimitDegrees=20, swing2LimitDegrees=20, twistLimitDegrees=20}
 {
  bone1Name=dyn_leg_03_bk_r
  bone2Name=dyn_leg_03_inbk_r
  swing1LimitDegrees=45
  swing1Motion=Free
  swing2LimitDegrees=45
  swing2Motion=Free
  twistLimitDegrees=45
  twistMotion=Free
 }
]
