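# Physics asset for the F_MED_SunBeam player body, covering a dangling five-link chain and a separate charm.
# Chain and charm each hang from a small master body constrained to the pelvis, with narrow swing and almost no twist between chain links, and looser limits on the charm.
# Larger capsules on the pelvis and both thighs give body volume around the dangling pieces, joined by angularly free constraints.
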PhysicsAsset /BRCosmetics/Characters/Player/Female/Medium/Bodies/F_MED_SunBeam/Meshes/F_MED_SunBeam_Physics
bodies {
 dyn_chain_00_MSTR [{shapeType=Capsule, length=0.505, radius=0.505, shapeName=dyn_chain_00_MSTR_capsule}]
 dyn_chain_01 [{shapeType=Capsule, length=5, radius=1, shapeName=dyn_chain_01_capsule, center=<4 0 0>, rotation=<rot 90 0 0>}]
 dyn_chain_02 [{shapeType=Capsule, length=5, radius=1, shapeName=dyn_chain_02_capsule, center=<4 0 0>, rotation=<rot 90 0 0>}]
 dyn_chain_03 [{shapeType=Capsule, length=5, radius=1, shapeName=dyn_chain_03_capsule, center=<4 0 0>, rotation=<rot 90 0 0>}]
 dyn_chain_04 [{shapeType=Capsule, length=5, radius=1, shapeName=dyn_chain_04_capsule, center=<4 0 0>, rotation=<rot 90 0 0>}]
 dyn_chain_05 [{shapeType=Capsule, length=5, radius=1, shapeName=dyn_chain_05_capsule, center=<4 0 0>, rotation=<rot 90 0 0>}]
 dyn_charm [{shapeType=Capsule, length=2, radius=1, shapeName=dyn_charm_capsule, center=<2.8324 -0.1 0>, rotation=<rot 0 0 -5>}]
 dyn_charm_MSTR [{shapeType=Capsule, length=0.505, radius=0.505, shapeName=dyn_charm_MSTR_capsule}]
 pelvis [{shapeType=Capsule, length=10, radius=10, shapeName=pelvis_capsule, center=<0.0432 0 0>}]
 root [{shapeType=Capsule, length=0.505, radius=0.505, shapeName=root_capsule}]
 thigh_l [{shapeType=Capsule, length=20, radius=9, shapeName=thigh_l_capsule, center=<-15 0 0>, rotation=<rot 90 0 0>}]
 thigh_r [{shapeType=Capsule, length=20, radius=9, shapeName=thigh_r_capsule, center=<15 0 0>, rotation=<rot 90 0 0>}]
}
constraints [
 {
  bone1Name=pelvis
  bone2Name=root
  swing1LimitDegrees=45
  swing1Motion=Free
  swing2LimitDegrees=45
  swing2Motion=Free
  twistLimitDegrees=45
  twistMotion=Free
 }
 {
  bone1Name=dyn_charm_MSTR
  bone2Name=pelvis
  swing1LimitDegrees=45
  swing1Motion=Free
  swing2LimitDegrees=45
  swing2Motion=Free
  twistLimitDegrees=45
  twistMotion=Free
 }
 {bone1Name=dyn_charm, bone2Name=dyn_charm_MSTR, swing1LimitDegrees=45, swing2LimitDegrees=45, twistLimitDegrees=45}
 {
  bone1Name=dyn_chain_00_MSTR
  bone2Name=pelvis
  swing1LimitDegrees=45
  swing1Motion=Free
  swing2LimitDegrees=45
  swing2Motion=Free
  twistLimitDegrees=45
  twistMotion=Free
 }
 {bone1Name=dyn_chain_01, bone2Name=dyn_chain_00_MSTR, swing1LimitDegrees=25, swing2LimitDegrees=25, twistLimitDegrees=1}
 {bone1Name=dyn_chain_02, bone2Name=dyn_chain_01, swing1LimitDegrees=25, swing2LimitDegrees=25, twistLimitDegrees=1}
 {bone1Name=dyn_chain_03, bone2Name=dyn_chain_02, swing1LimitDegrees=25, swing2LimitDegrees=25, twistLimitDegrees=1}
 {bone1Name=dyn_chain_04, bone2Name=dyn_chain_03, swing1LimitDegrees=25, swing2LimitDegrees=25, twistLimitDegrees=1}
 {bone1Name=dyn_chain_05, bone2Name=dyn_chain_04, swing1LimitDegrees=25, swing2LimitDegrees=25, twistLimitDegrees=1}
 {
  bone1Name=thigh_l
  bone2Name=pelvis
  swing1LimitDegrees=45
  swing1Motion=Free
  swing2LimitDegrees=45
  swing2Motion=Free
  twistLimitDegrees=45
  twistMotion=Free
 }
 {
  bone1Name=thigh_r
  bone2Name=pelvis
  swing1LimitDegrees=45
  swing1Motion=Free
  swing2LimitDegrees=45
  swing2Motion=Free
  twistLimitDegrees=45
  twistMotion=Free
 }
]
