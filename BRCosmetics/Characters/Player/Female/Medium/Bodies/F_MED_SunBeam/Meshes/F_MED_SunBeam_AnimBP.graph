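# Animation Blueprint for the SunBeam female medium body cosmetic, adding secondary motion to the pose copied from its attached parent mesh.
# The pose passes through deform and crouch-aware Control Rigs, pelvis-based RigidBody, then AnimDynamics, constraints and bone offsets for legs, straps, ties and accessories.
# The EventGraph derives alpha variables from crouch, movement and relaxed states.
# Two calf-based RigidBody nodes sit disconnected from the output.
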AnimBlueprint /BRCosmetics/Characters/Player/Female/Medium/Bodies/F_MED_SunBeam/Meshes/F_MED_SunBeam_AnimBP
target_skeleton=&/F_MED_SunBeam_Skeleton
graphs [
 AnimGraph {nodes [
   n1 = RigidBody() {
    bClampLinearTranslationLimitToRefPose=true
    bOverrideWorldGravity=true
    overridePhysicsAsset=&/F_MED_SunBeam_LegL_Physics
    simulationSpace=BaseBoneSpace
    alphaBoolBlend {blendOption=Linear}
    baseBoneRef {boneName=calf_l}
    componentAppliedLinearAccClamp=<2000 2000 2000>
    componentLinearAccScale=<0.5 0.5 0.75>
    componentLinearVelScale=<0.25 0.25 0.35>
    overrideWorldGravity {z=-980}
   }
   RigidBody(ComponentPose=@n1) {
    bClampLinearTranslationLimitToRefPose=true
    bOverrideWorldGravity=true
    overridePhysicsAsset=&/F_MED_SunBeam_LegR_Physics
    simulationSpace=BaseBoneSpace
    alphaBoolBlend {blendOption=Linear}
    baseBoneRef {boneName=calf_r}
    componentAppliedLinearAccClamp=<2000 2000 2000>
    componentLinearAccScale=<0.5 0.5 0.75>
    componentLinearVelScale=<0.25 0.25 0.35>
    overrideWorldGravity {z=-980}
   }
   n2 = Control Rig(Source=Copy Pose From Mesh() {bUseAttachedParent=true}) {
    alphaBoolBlend {blendOption=Linear}
    controlRigAssetReference {blueprintRigClass=&/F_MED_SunBeam_DfrmRig_CtrlRig.F_MED_SunBeam_DfrmRig_CtrlRig_C}
   }
   n3 = Control Rig!(Source=@n2, Alpha, isCrouching:bool=Get bIsCrouching()) -> (Pose) {
    alphaBoolBlend {blendOption=Linear}
    controlRigAssetReference {blueprintRigClass=&/F_MED_SunBeam_CtrlRig.F_MED_SunBeam_CtrlRig_C}
   }
   n4 = RigidBody(ComponentPose=Local To Component(LocalPose=@n3)) {
    bClampLinearTranslationLimitToRefPose=true
    bOverrideWorldGravity=true
    overridePhysicsAsset=&/F_MED_SunBeam_Physics
    simulationSpace=BaseBoneSpace
    alphaBoolBlend {blendOption=Linear}
    baseBoneRef {boneName=pelvis}
    componentAppliedLinearAccClamp=<2000 2000 2000>
    componentLinearAccScale=<0.4 0.3 0.3>
    componentLinearVelScale=<0.4 0.3 0.3>
    overrideWorldGravity {z=-980}
   }
   n5 = Anim Dynamics - Bone: dyn_leg_01_inbk_l(ComponentPose=@n4, Alpha=1, ExternalForce="0,0,0") {
    angularDampingOverride=0.9
    angularSpringConstant=150
    bAngularSpring=true
    bOverrideAngularDamping=true
    simulationSpace=BoneRelative
    alphaBoolBlend {blendOption=Linear}
    boundBone {boneName=dyn_leg_01_inbk_l}
    componentAppliedLinearAccClamp=<30000 30000 30000>
    componentLinearAccScale=<3 3 5>
    componentLinearVelScale=<3 3 5>
    physicsBodyDefinitions [{
      boundBone {boneName=dyn_leg_01_inbk_l}
      boxExtents=<25 25 25>
      constraintSetup {angularLimitsMax=<0 0 5>, angularLimitsMin=<0 -10 -5>, angularTarget=<0 0 0>}
      localJointOffset=<1 0 0>
     }]
    relativeSpaceBone {boneName=root}
   }
   n6 = Anim Dynamics - Bone: dyn_leg_02_inbk_l(ComponentPose=@n5, Alpha=1, ExternalForce="0,0,0") {
    angularSpringConstant=75
    bAngularSpring=true
    simulationSpace=BoneRelative
    alphaBoolBlend {blendOption=Linear}
    boundBone {boneName=dyn_leg_02_inbk_l}
    componentAppliedLinearAccClamp=<30000 30000 30000>
    componentLinearAccScale=<3 3 5>
    componentLinearVelScale=<3 3 5>
    physicsBodyDefinitions [{
      boundBone {boneName=dyn_leg_02_inbk_l}
      boxExtents=<25 25 25>
      constraintSetup {angularLimitsMax=<0 0 5>, angularLimitsMin=<0 -15 -5>, angularTarget=<0 0 0>}
      localJointOffset=<1 0 0>
     }]
    relativeSpaceBone {boneName=root}
   }
   n7 = Anim Dynamics - Bone: dyn_leg_03_inbk_l(ComponentPose=@n6, Alpha=1, ExternalForce="0,0,0") {
    angularSpringConstant=50
    bAngularSpring=true
    simulationSpace=BoneRelative
    alphaBoolBlend {blendOption=Linear}
    boundBone {boneName=dyn_leg_03_inbk_l}
    componentAppliedLinearAccClamp=<30000 30000 30000>
    componentLinearAccScale=<3 3 5>
    componentLinearVelScale=<3 3 5>
    physicsBodyDefinitions [{
      boundBone {boneName=dyn_leg_03_inbk_l}
      boxExtents=<25 25 25>
      constraintSetup {angularLimitsMax=<0 0 10>, angularLimitsMin=<0 -25 -10>, angularTarget=<0 0 0>}
      localJointOffset=<1 0 0>
     }]
    relativeSpaceBone {boneName=root}
   }
   n8 = Anim Dynamics - Bone: dyn_leg_01_bk_l(ComponentPose=@n7, Alpha=1, ExternalForce="0,0,0") {
    angularDampingOverride=0.9
    angularSpringConstant=150
    bAngularSpring=true
    bOverrideAngularDamping=true
    simulationSpace=BoneRelative
    alphaBoolBlend {blendOption=Linear}
    boundBone {boneName=dyn_leg_01_bk_l}
    componentAppliedLinearAccClamp=<30000 30000 30000>
    componentLinearAccScale=<3 3 5>
    componentLinearVelScale=<3 3 5>
    physicsBodyDefinitions [{
      boundBone {boneName=dyn_leg_01_bk_l}
      boxExtents=<25 25 25>
      constraintSetup {angularLimitsMax=<0 0 5>, angularLimitsMin=<0 -10 -5>, angularTarget=<0 0 0>}
      localJointOffset=<1 0 0>
     }]
    relativeSpaceBone {boneName=root}
   }
   n9 = Anim Dynamics - Bone: dyn_leg_02_bk_l(ComponentPose=@n8, Alpha=1, ExternalForce="0,0,0") {
    angularSpringConstant=75
    bAngularSpring=true
    simulationSpace=BoneRelative
    alphaBoolBlend {blendOption=Linear}
    boundBone {boneName=dyn_leg_02_bk_l}
    componentAppliedLinearAccClamp=<30000 30000 30000>
    componentLinearAccScale=<3 3 5>
    componentLinearVelScale=<3 3 5>
    physicsBodyDefinitions [{
      boundBone {boneName=dyn_leg_02_bk_l}
      boxExtents=<25 25 25>
      constraintSetup {angularLimitsMax=<0 0 5>, angularLimitsMin=<0 -15 -5>, angularTarget=<0 0 0>}
      localJointOffset=<1 0 0>
     }]
    relativeSpaceBone {boneName=root}
   }
   n10 = Anim Dynamics - Bone: dyn_leg_03_bk_l(ComponentPose=@n9, Alpha=1, ExternalForce="0,0,0") {
    angularSpringConstant=50
    bAngularSpring=true
    simulationSpace=BoneRelative
    alphaBoolBlend {blendOption=Linear}
    boundBone {boneName=dyn_leg_03_bk_l}
    componentAppliedLinearAccClamp=<30000 30000 30000>
    componentLinearAccScale=<3 3 5>
    componentLinearVelScale=<3 3 5>
    physicsBodyDefinitions [{
      boundBone {boneName=dyn_leg_03_bk_l}
      boxExtents=<25 25 25>
      constraintSetup {angularLimitsMax=<0 0 10>, angularLimitsMin=<0 -25 -10>, angularTarget=<0 0 0>}
      localJointOffset=<1 0 0>
     }]
    relativeSpaceBone {boneName=root}
   }
   n11 = Anim Dynamics - Bone: dyn_leg_01_in_l(ComponentPose=@n10, Alpha=1, ExternalForce="0,0,0") {
    angularDampingOverride=0.9
    angularSpringConstant=150
    bAngularSpring=true
    bOverrideAngularDamping=true
    simulationSpace=BoneRelative
    alphaBoolBlend {blendOption=Linear}
    boundBone {boneName=dyn_leg_01_in_l}
    componentAppliedLinearAccClamp=<30000 30000 30000>
    componentLinearAccScale=<3 3 5>
    componentLinearVelScale=<3 3 5>
    physicsBodyDefinitions [{
      boundBone {boneName=dyn_leg_01_in_l}
      boxExtents=<25 25 25>
      constraintSetup {angularLimitsMax=<0 0 5>, angularLimitsMin=<0 -10 -5>, angularTarget=<0 0 0>}
      localJointOffset=<1 0 0>
     }]
    relativeSpaceBone {boneName=root}
   }
   n12 = Anim Dynamics - Bone: dyn_leg_02_in_l(ComponentPose=@n11, Alpha=1, ExternalForce="0,0,0") {
    angularSpringConstant=75
    bAngularSpring=true
    simulationSpace=BoneRelative
    alphaBoolBlend {blendOption=Linear}
    boundBone {boneName=dyn_leg_02_in_l}
    componentAppliedLinearAccClamp=<30000 30000 30000>
    componentLinearAccScale=<3 3 5>
    componentLinearVelScale=<3 3 5>
    physicsBodyDefinitions [{
      boundBone {boneName=dyn_leg_02_in_l}
      boxExtents=<25 25 25>
      constraintSetup {angularLimitsMax=<0 0 5>, angularLimitsMin=<0 -15 -5>, angularTarget=<0 0 0>}
      localJointOffset=<1 0 0>
     }]
    relativeSpaceBone {boneName=root}
   }
   n13 = Anim Dynamics - Bone: dyn_leg_03_in_l(ComponentPose=@n12, Alpha=1, ExternalForce="0,0,0") {
    angularSpringConstant=50
    bAngularSpring=true
    simulationSpace=BoneRelative
    alphaBoolBlend {blendOption=Linear}
    boundBone {boneName=dyn_leg_03_in_l}
    componentAppliedLinearAccClamp=<30000 30000 30000>
    componentLinearAccScale=<3 3 5>
    componentLinearVelScale=<3 3 5>
    physicsBodyDefinitions [{
      boundBone {boneName=dyn_leg_03_in_l}
      boxExtents=<25 25 25>
      constraintSetup {angularLimitsMax=<0 0 10>, angularLimitsMin=<0 -25 -10>, angularTarget=<0 0 0>}
      localJointOffset=<1 0 0>
     }]
    relativeSpaceBone {boneName=root}
   }
   n14 = Anim Dynamics - Bone: dyn_leg_01_ft_l(ComponentPose=@n13, Alpha=1, ExternalForce="0,0,0") {
    angularDampingOverride=0.9
    angularSpringConstant=150
    bAngularSpring=true
    bOverrideAngularDamping=true
    simulationSpace=BoneRelative
    alphaBoolBlend {blendOption=Linear}
    boundBone {boneName=dyn_leg_01_ft_l}
    componentAppliedLinearAccClamp=<30000 30000 30000>
    componentLinearAccScale=<3 3 5>
    componentLinearVelScale=<3 3 5>
    physicsBodyDefinitions [{
      boundBone {boneName=dyn_leg_01_ft_l}
      boxExtents=<25 25 25>
      constraintSetup {angularLimitsMax=<0 0 5>, angularLimitsMin=<0 -10 -5>, angularTarget=<0 0 0>}
      localJointOffset=<1 0 0>
     }]
    relativeSpaceBone {boneName=root}
   }
   n15 = Anim Dynamics - Bone: dyn_leg_02_ft_l(ComponentPose=@n14, Alpha=1, ExternalForce="0,0,0") {
    angularSpringConstant=75
    bAngularSpring=true
    simulationSpace=BoneRelative
    alphaBoolBlend {blendOption=Linear}
    boundBone {boneName=dyn_leg_02_ft_l}
    componentAppliedLinearAccClamp=<30000 30000 30000>
    componentLinearAccScale=<3 3 5>
    componentLinearVelScale=<3 3 5>
    physicsBodyDefinitions [{
      boundBone {boneName=dyn_leg_02_ft_l}
      boxExtents=<25 25 25>
      constraintSetup {angularLimitsMax=<0 0 5>, angularLimitsMin=<0 -15 -5>, angularTarget=<0 0 0>}
      localJointOffset=<1 0 0>
     }]
    relativeSpaceBone {boneName=root}
   }
   n16 = Anim Dynamics - Bone: dyn_leg_03_ft_l(ComponentPose=@n15, Alpha=1, ExternalForce="0,0,0") {
    angularSpringConstant=50
    bAngularSpring=true
    simulationSpace=BoneRelative
    alphaBoolBlend {blendOption=Linear}
    boundBone {boneName=dyn_leg_03_ft_l}
    componentAppliedLinearAccClamp=<30000 30000 30000>
    componentLinearAccScale=<3 3 5>
    componentLinearVelScale=<3 3 5>
    physicsBodyDefinitions [{
      boundBone {boneName=dyn_leg_03_ft_l}
      boxExtents=<25 25 25>
      constraintSetup {angularLimitsMax=<0 0 10>, angularLimitsMin=<0 -25 -10>, angularTarget=<0 0 0>}
      localJointOffset=<1 0 0>
     }]
    relativeSpaceBone {boneName=root}
   }
   n17 = Anim Dynamics - Bone: dyn_leg_01_ft_r(ComponentPose=@n16, Alpha=1, ExternalForce="0,0,0") {
    angularDampingOverride=0.9
    angularSpringConstant=250
    bAngularSpring=true
    bOverrideAngularDamping=true
    simulationSpace=BoneRelative
    alphaBoolBlend {blendOption=Linear}
    boundBone {boneName=dyn_leg_01_ft_r}
    componentAppliedLinearAccClamp=<30000 30000 30000>
    componentLinearAccScale=<3 3 5>
    componentLinearVelScale=<3 3 5>
    physicsBodyDefinitions [{
      boundBone {boneName=dyn_leg_01_ft_r}
      boxExtents=<25 25 25>
      constraintSetup {angularLimitsMax=<0 0 5>, angularLimitsMin=<0 -10 -5>, angularTarget=<0 0 0>}
      localJointOffset=<1 0 0>
     }]
    relativeSpaceBone {boneName=root}
   }
   n18 = Anim Dynamics - Bone: dyn_leg_02_ft_r(ComponentPose=@n17, Alpha=1, ExternalForce="0,0,0") {
    angularSpringConstant=75
    bAngularSpring=true
    simulationSpace=BoneRelative
    alphaBoolBlend {blendOption=Linear}
    boundBone {boneName=dyn_leg_02_ft_r}
    componentAppliedLinearAccClamp=<30000 30000 30000>
    componentLinearAccScale=<3 3 5>
    componentLinearVelScale=<3 3 5>
    physicsBodyDefinitions [{
      boundBone {boneName=dyn_leg_02_ft_r}
      boxExtents=<25 25 25>
      constraintSetup {angularLimitsMax=<0 0 5>, angularLimitsMin=<0 -15 -5>, angularTarget=<0 0 0>}
      localJointOffset=<1 0 0>
     }]
    relativeSpaceBone {boneName=root}
   }
   n19 = Anim Dynamics - Bone: dyn_leg_03_ft_r(ComponentPose=@n18, Alpha=1, ExternalForce="0,0,0") {
    angularSpringConstant=50
    bAngularSpring=true
    simulationSpace=BoneRelative
    alphaBoolBlend {blendOption=Linear}
    boundBone {boneName=dyn_leg_03_ft_r}
    componentAppliedLinearAccClamp=<30000 30000 30000>
    componentLinearAccScale=<3 3 5>
    componentLinearVelScale=<3 3 5>
    physicsBodyDefinitions [{
      boundBone {boneName=dyn_leg_03_ft_r}
      boxExtents=<25 25 25>
      constraintSetup {angularLimitsMax=<0 0 10>, angularLimitsMin=<0 -25 -10>, angularTarget=<0 0 0>}
      localJointOffset=<1 0 0>
     }]
    relativeSpaceBone {boneName=root}
   }
   n20 = Anim Dynamics - Bone: dyn_leg_01_in_r(ComponentPose=@n19, Alpha=1, ExternalForce="0,0,0") {
    angularDampingOverride=0.999
    angularSpringConstant=250
    bAngularSpring=true
    bOverrideAngularDamping=true
    simulationSpace=BoneRelative
    alphaBoolBlend {blendOption=Linear}
    boundBone {boneName=dyn_leg_01_in_r}
    componentAppliedLinearAccClamp=<30000 30000 30000>
    componentLinearAccScale=<3 3 5>
    componentLinearVelScale=<3 3 5>
    physicsBodyDefinitions [{
      boundBone {boneName=dyn_leg_01_in_r}
      boxExtents=<25 25 25>
      constraintSetup {angularLimitsMax=<0 0 5>, angularLimitsMin=<0 -10 -5>, angularTarget=<0 0 0>}
      localJointOffset=<1 0 0>
     }]
    relativeSpaceBone {boneName=root}
   }
   n21 = Anim Dynamics - Bone: dyn_leg_02_in_r(ComponentPose=@n20, Alpha=1, ExternalForce="0,0,0") {
    angularDampingOverride=0.9
    angularSpringConstant=150
    bAngularSpring=true
    bOverrideAngularDamping=true
    simulationSpace=BoneRelative
    alphaBoolBlend {blendOption=Linear}
    boundBone {boneName=dyn_leg_02_in_r}
    componentAppliedLinearAccClamp=<30000 30000 30000>
    componentLinearAccScale=<3 3 5>
    componentLinearVelScale=<3 3 5>
    physicsBodyDefinitions [{
      boundBone {boneName=dyn_leg_02_in_r}
      boxExtents=<25 25 25>
      constraintSetup {angularLimitsMax=<0 0 5>, angularLimitsMin=<0 -15 -5>, angularTarget=<0 0 0>}
      localJointOffset=<1 0 0>
     }]
    relativeSpaceBone {boneName=root}
   }
   n22 = Anim Dynamics - Bone: dyn_leg_03_in_r(ComponentPose=@n21, Alpha=1, ExternalForce="0,0,0") {
    angularSpringConstant=50
    bAngularSpring=true
    simulationSpace=BoneRelative
    alphaBoolBlend {blendOption=Linear}
    boundBone {boneName=dyn_leg_03_in_r}
    componentAppliedLinearAccClamp=<30000 30000 30000>
    componentLinearAccScale=<3 3 5>
    componentLinearVelScale=<3 3 5>
    physicsBodyDefinitions [{
      boundBone {boneName=dyn_leg_03_in_r}
      boxExtents=<25 25 25>
      constraintSetup {angularLimitsMax=<0 0 10>, angularLimitsMin=<0 -25 -10>, angularTarget=<0 0 0>}
      localJointOffset=<1 0 0>
     }]
    relativeSpaceBone {boneName=root}
   }
   n23 = Anim Dynamics - Bone: dyn_leg_01_inbk_r(ComponentPose=@n22, Alpha=1, ExternalForce="0,0,0") {
    angularDampingOverride=0.999
    angularSpringConstant=250
    bAngularSpring=true
    bOverrideAngularDamping=true
    simulationSpace=BoneRelative
    alphaBoolBlend {blendOption=Linear}
    boundBone {boneName=dyn_leg_01_inbk_r}
    componentAppliedLinearAccClamp=<30000 30000 30000>
    componentLinearAccScale=<3 3 5>
    componentLinearVelScale=<3 3 5>
    physicsBodyDefinitions [{
      boundBone {boneName=dyn_leg_01_inbk_r}
      boxExtents=<25 25 25>
      constraintSetup {angularLimitsMax=<0 0 5>, angularLimitsMin=<0 -10 -5>, angularTarget=<0 0 0>}
      localJointOffset=<1 0 0>
     }]
    relativeSpaceBone {boneName=root}
   }
   n24 = Anim Dynamics - Bone: dyn_leg_02_inbk_r(ComponentPose=@n23, Alpha=1, ExternalForce="0,0,0") {
    angularBiasOverride=0.9
    angularSpringConstant=150
    bAngularSpring=true
    bOverrideAngularBias=true
    simulationSpace=BoneRelative
    alphaBoolBlend {blendOption=Linear}
    boundBone {boneName=dyn_leg_02_inbk_r}
    componentAppliedLinearAccClamp=<30000 30000 30000>
    componentLinearAccScale=<3 3 5>
    componentLinearVelScale=<3 3 5>
    physicsBodyDefinitions [{
      boundBone {boneName=dyn_leg_02_inbk_r}
      boxExtents=<25 25 25>
      constraintSetup {angularLimitsMax=<0 0 5>, angularLimitsMin=<0 -15 -5>, angularTarget=<0 0 0>}
      localJointOffset=<1 0 0>
     }]
    relativeSpaceBone {boneName=root}
   }
   n25 = Anim Dynamics - Bone: dyn_leg_03_inbk_r(ComponentPose=@n24, Alpha=1, ExternalForce="0,0,0") {
    angularSpringConstant=50
    bAngularSpring=true
    simulationSpace=BoneRelative
    alphaBoolBlend {blendOption=Linear}
    boundBone {boneName=dyn_leg_03_inbk_r}
    componentAppliedLinearAccClamp=<30000 30000 30000>
    componentLinearAccScale=<3 3 5>
    componentLinearVelScale=<3 3 5>
    physicsBodyDefinitions [{
      boundBone {boneName=dyn_leg_03_inbk_r}
      boxExtents=<25 25 25>
      constraintSetup {angularLimitsMax=<0 0 10>, angularLimitsMin=<0 -25 -10>, angularTarget=<0 0 0>}
      localJointOffset=<1 0 0>
     }]
    relativeSpaceBone {boneName=root}
   }
   n26 = Anim Dynamics - Bone: dyn_leg_01_bk_r(ComponentPose=@n25, Alpha=1, ExternalForce="0,0,0") {
    angularDampingOverride=0.9
    angularSpringConstant=250
    bAngularSpring=true
    bOverrideAngularDamping=true
    simulationSpace=BoneRelative
    alphaBoolBlend {blendOption=Linear}
    boundBone {boneName=dyn_leg_01_bk_r}
    componentAppliedLinearAccClamp=<30000 30000 30000>
    componentLinearAccScale=<3 3 5>
    componentLinearVelScale=<3 3 5>
    physicsBodyDefinitions [{
      boundBone {boneName=dyn_leg_01_bk_r}
      boxExtents=<25 25 25>
      constraintSetup {angularLimitsMax=<0 0 5>, angularLimitsMin=<0 -10 -5>, angularTarget=<0 0 0>}
      localJointOffset=<1 0 0>
     }]
    relativeSpaceBone {boneName=root}
   }
   n27 = Anim Dynamics - Bone: dyn_leg_02_bk_r(ComponentPose=@n26, Alpha=1, ExternalForce="0,0,0") {
    angularSpringConstant=75
    bAngularSpring=true
    simulationSpace=BoneRelative
    alphaBoolBlend {blendOption=Linear}
    boundBone {boneName=dyn_leg_02_bk_r}
    componentAppliedLinearAccClamp=<30000 30000 30000>
    componentLinearAccScale=<3 3 5>
    componentLinearVelScale=<3 3 5>
    physicsBodyDefinitions [{
      boundBone {boneName=dyn_leg_02_bk_r}
      boxExtents=<25 25 25>
      constraintSetup {angularLimitsMax=<0 0 5>, angularLimitsMin=<0 -15 -5>, angularTarget=<0 0 0>}
      localJointOffset=<1 0 0>
     }]
    relativeSpaceBone {boneName=root}
   }
   n28 = Anim Dynamics - Bone: dyn_leg_03_bk_r(ComponentPose=@n27, Alpha=1, ExternalForce="0,0,0") {
    angularSpringConstant=50
    bAngularSpring=true
    simulationSpace=BoneRelative
    alphaBoolBlend {blendOption=Linear}
    boundBone {boneName=dyn_leg_03_bk_r}
    componentAppliedLinearAccClamp=<30000 30000 30000>
    componentLinearAccScale=<3 3 5>
    componentLinearVelScale=<3 3 5>
    physicsBodyDefinitions [{
      boundBone {boneName=dyn_leg_03_bk_r}
      boxExtents=<25 25 25>
      constraintSetup {angularLimitsMax=<0 0 10>, angularLimitsMin=<0 -25 -10>, angularTarget=<0 0 0>}
      localJointOffset=<1 0 0>
     }]
    relativeSpaceBone {boneName=root}
   }
   n29 = "Constraint - dyn_leg_01_ft_r (P)"(ComponentPose=@n28, Alpha=1, ConstraintWeights_0=1) {
    alphaBoolBlend {blendOption=Linear}
    alphaScaleBiasClamp {bInterpResult=true, interpSpeedDecreasing=20, interpSpeedIncreasing=20}
    boneToModify {boneName=dyn_leg_01_ft_r}
    constraintSetup [{targetBone {boneName=dyn_leg_00_ft_r_MSTR}}]
    constraintWeights [1]
   }
   n30 = "Constraint - dyn_leg_02_ft_r (P)"(ComponentPose=@n29, Alpha=1, ConstraintWeights_0=1) {
    alphaBoolBlend {blendOption=Linear}
    alphaScaleBiasClamp {bInterpResult=true, interpSpeedDecreasing=20, interpSpeedIncreasing=20}
    boneToModify {boneName=dyn_leg_02_ft_r}
    constraintSetup [{targetBone {boneName=dyn_leg_00_ft_r_MSTR}}]
    constraintWeights [1]
   }
   n31 = "Constraint - dyn_leg_03_ft_r (P)"(ComponentPose=@n30, Alpha=1, ConstraintWeights_0=1) {
    alphaBoolBlend {blendOption=Linear}
    alphaScaleBiasClamp {bInterpResult=true, interpSpeedDecreasing=20, interpSpeedIncreasing=20}
    boneToModify {boneName=dyn_leg_03_ft_r}
    constraintSetup [{targetBone {boneName=dyn_leg_00_ft_r_MSTR}}]
    constraintWeights [1]
   }
   n32 = "Transform (Modify) Bone - Bone: dyn_leg_00_ft_r_MSTR"(ComponentPose=@n31, Alpha=1, Translation="0,0,0", Rotation="0.5,0,-15", Scale="1,1,1") {
    rotationMode=BMM_Additive
    rotationSpace=BCS_BoneSpace
    alphaBoolBlend {blendOption=Linear}
    alphaScaleBiasClamp {bInterpResult=true, interpSpeedDecreasing=20, interpSpeedIncreasing=20}
    boneToModify {boneName=dyn_leg_00_ft_r_MSTR}
    rotation {pitch=0.5, roll=-15}
   }
   n33 = "Constraint - dyn_leg_01_in_r (P)"(ComponentPose=@n32, Alpha=1, ConstraintWeights_0=1) {
    alphaBoolBlend {blendOption=Linear}
    alphaScaleBiasClamp {bInterpResult=true, interpSpeedDecreasing=20, interpSpeedIncreasing=20}
    boneToModify {boneName=dyn_leg_01_in_r}
    constraintSetup [{targetBone {boneName=dyn_leg_00_in_r_MSTR}}]
    constraintWeights [1]
   }
   n34 = "Constraint - dyn_leg_02_in_r (P)"(ComponentPose=@n33, Alpha=1, ConstraintWeights_0=1) {
    alphaBoolBlend {blendOption=Linear}
    alphaScaleBiasClamp {bInterpResult=true, interpSpeedDecreasing=20, interpSpeedIncreasing=20}
    boneToModify {boneName=dyn_leg_02_in_r}
    constraintSetup [{targetBone {boneName=dyn_leg_00_in_r_MSTR}}]
    constraintWeights [1]
   }
   n35 = "Constraint - dyn_leg_03_in_r (P)"(ComponentPose=@n34, Alpha=1, ConstraintWeights_0=1) {
    alphaBoolBlend {blendOption=Linear}
    alphaScaleBiasClamp {bInterpResult=true, interpSpeedDecreasing=20, interpSpeedIncreasing=20}
    boneToModify {boneName=dyn_leg_03_in_r}
    constraintSetup [{targetBone {boneName=dyn_leg_00_in_r_MSTR}}]
    constraintWeights [1]
   }
   n36 = "Transform (Modify) Bone - Bone: dyn_necklace_MSTR"(ComponentPose=@n35, Alpha=1, Translation="0,0,0.1", Rotation="0,0,0", Scale="1,1,1") {
    rotationMode=BMM_Additive
    rotationSpace=BCS_BoneSpace
    translationMode=BMM_Additive
    translationSpace=BCS_BoneSpace
    alphaBoolBlend {blendOption=Linear}
    boneToModify {boneName=dyn_necklace_MSTR}
    translation {z=0.1}
   }
   n37 = "Transform (Modify) Bone - Bone: dyn_necklace_MSTR"(ComponentPose=@n36, Alpha=1, Translation="0,0,-0.1", Rotation="3,-20,0", Scale="1,1,1") {
    rotationMode=BMM_Additive
    rotationSpace=BCS_BoneSpace
    translationMode=BMM_Additive
    translationSpace=BCS_BoneSpace
    alphaBoolBlend {blendOption=Linear}
    boneToModify {boneName=dyn_necklace_MSTR}
    rotation {pitch=3, yaw=-20}
    translation {z=-0.1}
   }
   n38 = Anim Dynamics - Bone: dyn_necklace(ComponentPose=@n37, Alpha=1, ExternalForce="0,0,0") {
    angularDampingOverride=0.97
    angularSpringConstant=20000
    bAngularSpring=true
    bOverrideAngularDamping=true
    bUseGravityOverride=true
    simulationSpace=RootRelative
    alphaBoolBlend {blendOption=Linear}
    boundBone {boneName=dyn_necklace}
    componentAppliedLinearAccClamp=<2000 2000 2000>
    componentLinearAccScale=<0.3 0.3 0.3>
    componentLinearVelScale=<0.3 0.3 0.3>
    gravityOverride {z=-980}
    physicsBodyDefinitions [{
      boundBone {boneName=dyn_necklace}
      boxExtents=<11 21 21>
      constraintSetup {angularLimitsMax=<1 0 15>, angularLimitsMin=<-1 -15 -15>, angularTarget=<0 0 0>}
      localJointOffset=<4 0 0>
     }]
   }
   n39 = "Constraint - dyn_pouch_MOVE (P)"(ComponentPose=@n38, Alpha=1, ConstraintWeights_0=1) {
    alphaBoolBlend {blendOption=Linear}
    boneToModify {boneName=dyn_pouch_MOVE}
    constraintSetup [{offsetOption=None, targetBone {boneName=dyn_belt_l_MSTR}}]
    constraintWeights [1]
   }
   n40 = "Transform (Modify) Bone - Bone: dyn_pouchStrap_l"(ComponentPose=@n39, Alpha=1, Translation="-1,0,0", Rotation="0,0,0", Scale="1,1,1") {
    rotationMode=BMM_Additive
    translationMode=BMM_Additive
    alphaBoolBlend {blendOption=Linear}
    boneToModify {boneName=dyn_pouchStrap_l}
    translation {x=-1}
   }
   n41 = "Transform (Modify) Bone - Bone: dyn_pouchStrap_r"(ComponentPose=@n40, Alpha=1, Translation="0,0,0", Rotation="0,0,0", Scale="1,1,1") {rotationMode=BMM_Additive, translationMode=BMM_Additive, alphaBoolBlend {blendOption=Linear}, boneToModify {boneName=dyn_pouchStrap_r}}
   n42 = Look At - Bone: dyn_pouch_MSTR(ComponentPose=@n41, Alpha=1) {
    alphaBoolBlend {blendOption=Linear}
    boneToModify {boneName=dyn_pouch_MSTR}
    lookAtLocation=<0 15 -15>
    lookAtTarget {boneReference {boneName=calf_l}}
    lookAt_Axis {axis=<1 0>}
   }
   n43 = Anim Dynamics - Bone: dyn_pouch(ComponentPose=@n42, Alpha=1, ExternalForce="0,0,0") {
    angularDampingOverride=0.9999
    angularSpringConstant=20000
    bAngularSpring=true
    bOverrideAngularDamping=true
    bUseGravityOverride=true
    simulationSpace=RootRelative
    alphaBoolBlend {blendOption=Linear}
    boundBone {boneName=dyn_pouch}
    componentAppliedLinearAccClamp=<2000 2000 2000>
    componentLinearAccScale=<0.3 0.3 0.3>
    componentLinearVelScale=<0.3 0.3 0.3>
    gravityOverride {z=-980}
    physicsBodyDefinitions [{
      boundBone {boneName=dyn_pouch}
      boxExtents=<10 20 20>
      constraintSetup {angularLimitsMax=<1 1 15>, angularLimitsMin=<-1 -15 -15>, angularTarget=<0 0 0>}
      localJointOffset=<4 0 0>
     }]
   }
   n44 = Anim Dynamics - Bone: dyn_hipLoop_l(ComponentPose=@n43, Alpha=1, ExternalForce="0,0,0") {
    angularDampingOverride=1
    angularSpringConstant=20000
    bAngularSpring=true
    bOverrideAngularDamping=true
    bUseGravityOverride=true
    simulationSpace=RootRelative
    alphaBoolBlend {blendOption=Linear}
    boundBone {boneName=dyn_hipLoop_l}
    componentAppliedLinearAccClamp=<2000 2000 2000>
    componentLinearAccScale=<0.3 0.3 0.3>
    componentLinearVelScale=<0.3 0.3 0.3>
    gravityOverride {z=-980}
    physicsBodyDefinitions [{
      boundBone {boneName=dyn_hipLoop_l}
      boxExtents=<11 21 21>
      constraintSetup {angularLimitsMax=<1 1 15>, angularLimitsMin=<-1 -15 -15>, angularTarget=<0 0 0>}
      localJointOffset=<4 0 0>
     }]
   }
   n45 = Anim Dynamics - Bone: dyn_hipLoop_r(ComponentPose=@n44, Alpha=1, ExternalForce="0,0,0") {
    angularDampingOverride=1
    angularSpringConstant=20000
    bAngularSpring=true
    bOverrideAngularDamping=true
    bUseGravityOverride=true
    simulationSpace=RootRelative
    alphaBoolBlend {blendOption=Linear}
    boundBone {boneName=dyn_hipLoop_r}
    componentAppliedLinearAccClamp=<2000 2000 2000>
    componentLinearAccScale=<0.3 0.3 0.3>
    componentLinearVelScale=<0.3 0.3 0.3>
    gravityOverride {z=-980}
    physicsBodyDefinitions [{
      boundBone {boneName=dyn_hipLoop_r}
      boxExtents=<11 21 21>
      constraintSetup {angularLimitsMax=<1 1 15>, angularLimitsMin=<-1 -15 -15>, angularTarget=<0 0 0>}
      localJointOffset=<4 0 0>
     }]
   }
   n46 = "Constraint - dyn_legStrap_l_MSTR (PPPPP)"!(
    ComponentPose=@n45
    Alpha=1
    ConstraintWeights_0=0.156
    ConstraintWeights_1:float=0.661
    ConstraintWeights_2:float=0.132
    ConstraintWeights_3:float=0.024
    ConstraintWeights_4:float=0.027
   ) -> (Pose) {
    alphaBoolBlend {blendOption=Linear}
    boneToModify {boneName=dyn_legStrap_l_MSTR}
    constraintSetup [
     {targetBone {boneName=pelvis}}
     {targetBone {boneName=thigh_twist_01_l}}
     {targetBone {boneName=thigh_l}}
     {targetBone {boneName=deform_glute_l}}
     {targetBone {boneName=deform_groin_l}}
    ]
    constraintWeights [1,1,1,1,1]
   }
   n47 = Anim Dynamics - Bone: dyn_legStrap_l(ComponentPose=@n46, Alpha=1, ExternalForce="0,0,0") {
    angularDampingOverride=1
    angularSpringConstant=20000
    bAngularSpring=true
    bOverrideAngularDamping=true
    bUseGravityOverride=true
    simulationSpace=RootRelative
    alphaBoolBlend {blendOption=Linear}
    boundBone {boneName=dyn_legStrap_l}
    componentAppliedLinearAccClamp=<2000 2000 2000>
    componentLinearAccScale=<0.3 0.3 0.3>
    componentLinearVelScale=<0.3 0.3 0.3>
    gravityOverride {z=-980}
    physicsBodyDefinitions [{
      boundBone {boneName=dyn_legStrap_l}
      boxExtents=<11 21 21>
      constraintSetup {angularLimitsMax=<1 1 15>, angularLimitsMin=<-1 -15 -15>, angularTarget=<0 0 0>}
      localJointOffset=<4 0 0>
     }]
   }
   n48 = "Constraint - dyn_legStrap_r_MSTR (PPPPP)"!(
    ComponentPose=@n47
    Alpha=1
    ConstraintWeights_0=0.156
    ConstraintWeights_1:float=0.661
    ConstraintWeights_2:float=0.132
    ConstraintWeights_3:float=0.024
    ConstraintWeights_4:float=0.027
   ) -> (Pose) {
    alphaBoolBlend {blendOption=Linear}
    boneToModify {boneName=dyn_legStrap_r_MSTR}
    constraintSetup [
     {targetBone {boneName=pelvis}}
     {targetBone {boneName=thigh_twist_01_r}}
     {targetBone {boneName=thigh_r}}
     {targetBone {boneName=deform_glute_r}}
     {targetBone {boneName=deform_groin_r}}
    ]
    constraintWeights [1,1,1,1,1]
   }
   n49 = Anim Dynamics - Bone: dyn_legStrap_r(ComponentPose=@n48, Alpha=1, ExternalForce="0,0,0") {
    angularDampingOverride=1
    angularSpringConstant=20000
    bAngularSpring=true
    bOverrideAngularDamping=true
    bUseGravityOverride=true
    simulationSpace=RootRelative
    alphaBoolBlend {blendOption=Linear}
    boundBone {boneName=dyn_legStrap_r}
    componentAppliedLinearAccClamp=<2000 2000 2000>
    componentLinearAccScale=<0.3 0.3 0.3>
    componentLinearVelScale=<0.3 0.3 0.3>
    gravityOverride {z=-980}
    physicsBodyDefinitions [{
      boundBone {boneName=dyn_legStrap_r}
      boxExtents=<11 21 21>
      constraintSetup {angularLimitsMax=<1 1 15>, angularLimitsMin=<-1 -15 -15>, angularTarget=<0 0 0>}
      localJointOffset=<4 0 0>
     }]
   }
   n50 = "Constraint - dyn_hipStraps_MSTR (PPPPP)"!(
    ComponentPose=@n49
    Alpha=1
    ConstraintWeights_0=0.027
    ConstraintWeights_1:float=0.866
    ConstraintWeights_2:float=0.092
    ConstraintWeights_3:float=0.005
    ConstraintWeights_4:float=0.01
   ) -> (Pose) {
    alphaBoolBlend {blendOption=Linear}
    boneToModify {boneName=dyn_hipStraps_MSTR}
    constraintSetup [
     {targetBone {boneName=spine_02}}
     {targetBone {boneName=pelvis}}
     {targetBone {boneName=thigh_twist_01_r}}
     {targetBone {boneName=deform_glute_r}}
     {targetBone {boneName=deform_groin_r}}
    ]
    constraintWeights [1,1,1,1,1]
   }
   n51 = "Constraint - dyn_hipStraps_MSTR (R)"(ComponentPose=@n50, Alpha=1, ConstraintWeights_0=1) {
    alphaBoolBlend {blendOption=Linear}
    alphaScaleBiasClamp {bInterpResult=true}
    boneToModify {boneName=dyn_hipStraps_MSTR}
    constraintSetup [{transformType=Rotation, targetBone {boneName=thigh_r}}]
    constraintWeights [1]
   }
   n52 = "Transform (Modify) Bone - Bone: dyn_hipStraps_MSTR"(ComponentPose=@n51, Alpha=1, Translation="0,0,0", Rotation="30,0,0", Scale="1,1,1") {
    rotationMode=BMM_Additive
    rotationSpace=BCS_BoneSpace
    alphaBoolBlend {blendOption=Linear}
    alphaScaleBiasClamp {bInterpResult=true}
    boneToModify {boneName=dyn_hipStraps_MSTR}
    rotation {pitch=30}
   }
   n53 = Anim Dynamics - Bone: dyn_hipStraps(ComponentPose=@n52, Alpha=1, ExternalForce="0,0,0") {
    angularDampingOverride=1
    angularSpringConstant=20000
    bAngularSpring=true
    bOverrideAngularDamping=true
    bUseGravityOverride=true
    simulationSpace=RootRelative
    alphaBoolBlend {blendOption=Linear}
    boundBone {boneName=dyn_hipStraps}
    componentAppliedLinearAccClamp=<2000 2000 2000>
    componentLinearAccScale=<0.3 0.3 0.3>
    componentLinearVelScale=<0.3 0.3 0.3>
    gravityOverride {z=-980}
    physicsBodyDefinitions [{
      boundBone {boneName=dyn_hipStraps}
      boxExtents=<11 21 21>
      constraintSetup {angularLimitsMax=<1 1 15>, angularLimitsMin=<-1 -15 -15>, angularTarget=<0 0 0>}
      localJointOffset=<4 0 0>
     }]
   }
   n54 = "Constraint - dyn_armTie_01_MSTR (PPP)"!(ComponentPose=@n53, Alpha=1, ConstraintWeights_0=0.033, ConstraintWeights_1:float=0.037, ConstraintWeights_2:float=0.93) -> (Pose) {
    alphaBoolBlend {blendOption=Linear}
    boneToModify {boneName=dyn_armTie_01_MSTR}
    constraintSetup [{targetBone {boneName=upperarm_l}}, {targetBone {boneName=upperarm_twist_02_l}}, {targetBone {boneName=upperarm_twist_01_l}}]
    constraintWeights [1,1,1]
   }
   n55 = Anim Dynamics - Bone: dyn_armTie_01(ComponentPose=@n54, Alpha=1, ExternalForce="0,0,0") {
    angularDampingOverride=1
    angularSpringConstant=20000
    bAngularSpring=true
    bOverrideAngularDamping=true
    bUseGravityOverride=true
    simulationSpace=RootRelative
    alphaBoolBlend {blendOption=Linear}
    boundBone {boneName=dyn_armTie_01}
    componentAppliedLinearAccClamp=<2000 2000 2000>
    componentLinearAccScale=<0.3 0.3 0.3>
    componentLinearVelScale=<0.3 0.3 0.3>
    gravityOverride {z=-980}
    physicsBodyDefinitions [{
      boundBone {boneName=dyn_armTie_01}
      boxExtents=<11 21 21>
      constraintSetup {angularLimitsMax=<1 0 15>, angularLimitsMin=<-1 -15 -15>, angularTarget=<0 0 0>}
      localJointOffset=<4 0 0>
     }]
   }
   n56 = "Constraint - dyn_armTie_02_MSTR (PPP)"!(ComponentPose=@n55, Alpha=1, ConstraintWeights_0=0.3, ConstraintWeights_1:float=0.252, ConstraintWeights_2:float=0.448) -> (Pose) {
    alphaBoolBlend {blendOption=Linear}
    boneToModify {boneName=dyn_armTie_02_MSTR}
    constraintSetup [{targetBone {boneName=upperarm_l}}, {targetBone {boneName=upperarm_twist_02_l}}, {targetBone {boneName=upperarm_twist_01_l}}]
    constraintWeights [1,1,1]
   }
   n57 = Anim Dynamics - Bone: dyn_armTie_02(ComponentPose=@n56, Alpha=1, ExternalForce="0,0,0") {
    angularDampingOverride=1
    angularSpringConstant=20000
    bAngularSpring=true
    bOverrideAngularDamping=true
    bUseGravityOverride=true
    simulationSpace=RootRelative
    alphaBoolBlend {blendOption=Linear}
    boundBone {boneName=dyn_armTie_02}
    componentAppliedLinearAccClamp=<2000 2000 2000>
    componentLinearAccScale=<0.3 0.3 0.3>
    componentLinearVelScale=<0.3 0.3 0.3>
    gravityOverride {z=-980}
    physicsBodyDefinitions [{
      boundBone {boneName=dyn_armTie_02}
      boxExtents=<11 21 21>
      constraintSetup {angularLimitsMax=<1 1 15>, angularLimitsMin=<-1 -15 -15>, angularTarget=<0 0 0>}
      localJointOffset=<4 0 0>
     }]
   }
   n58 = "Constraint - dyn_armTie_03_MSTR (PPPP)"!(
    ComponentPose=@n57
    Alpha=1
    ConstraintWeights_0=0.139
    ConstraintWeights_1:float=0.17
    ConstraintWeights_2:float=0.668
    ConstraintWeights_3:float=0.023
   ) -> (Pose) {
    alphaBoolBlend {blendOption=Linear}
    boneToModify {boneName=dyn_armTie_03_MSTR}
    constraintSetup [
     {targetBone {boneName=lowerarm_twist_01_l}}
     {targetBone {boneName=lowerarm_l}}
     {targetBone {boneName=lowerarm_twist_02_l}}
     {targetBone {boneName=upperarm_l}}
    ]
    constraintWeights [1,1,1,1]
   }
   n59 = Anim Dynamics - Bone: dyn_armTie_03(ComponentPose=@n58, Alpha=1, ExternalForce="0,0,0") {
    angularDampingOverride=1
    angularSpringConstant=20000
    bAngularSpring=true
    bOverrideAngularDamping=true
    bUseGravityOverride=true
    simulationSpace=RootRelative
    alphaBoolBlend {blendOption=Linear}
    boundBone {boneName=dyn_armTie_03}
    componentAppliedLinearAccClamp=<2000 2000 2000>
    componentLinearAccScale=<0.3 0.3 0.3>
    componentLinearVelScale=<0.3 0.3 0.3>
    gravityOverride {z=-980}
    physicsBodyDefinitions [{
      boundBone {boneName=dyn_armTie_03}
      boxExtents=<11 21 21>
      constraintSetup {angularLimitsMax=<1 1 15>, angularLimitsMin=<-1 -15 -15>, angularTarget=<0 0 0>}
      localJointOffset=<4 0 0>
     }]
   }
   n60 = "Transform (Modify) Bone - Bone: dyn_collar"(ComponentPose=@n59, Alpha=1, Translation="0,0,0", Rotation="0,0,0", Scale="1.05,1.05,1.05") {
    scaleMode=BMM_Additive
    scaleSpace=BCS_BoneSpace
    alphaBoolBlend {blendOption=Linear}
    boneToModify {boneName=dyn_collar}
    scale=<1.05 1.05 1.05>
   }
   n61 = "Transform (Modify) Bone - Bone: dyn_legL_bowL_MSTR"(ComponentPose=@n60, Alpha=1, Translation="0,0,0", Rotation="-5,-45,45", Scale="1,1,1") {
    rotationMode=BMM_Additive
    rotationSpace=BCS_BoneSpace
    alphaBoolBlend {blendOption=Linear}
    boneToModify {boneName=dyn_legL_bowL_MSTR}
    rotation=<rot -5 -45 45>
   }
   n62 = "Transform (Modify) Bone - Bone: dyn_legL_bowR_MSTR"(ComponentPose=@n61, Alpha=1, Translation="0,0,0", Rotation="-15,30,-15", Scale="1,1,1") {
    rotationMode=BMM_Additive
    rotationSpace=BCS_BoneSpace
    alphaBoolBlend {blendOption=Linear}
    boneToModify {boneName=dyn_legL_bowR_MSTR}
    rotation=<rot -15 30 -15>
   }
   n63 = "Transform (Modify) Bone - Bone: dyn_legR_bowL_MSTR"(ComponentPose=@n62, Alpha=1, Translation="0,0,0", Rotation="0,-30,0", Scale="1,1,1") {
    rotationMode=BMM_Additive
    rotationSpace=BCS_BoneSpace
    alphaBoolBlend {blendOption=Linear}
    boneToModify {boneName=dyn_legR_bowL_MSTR}
    rotation {yaw=-30}
   }
   n64 = "Transform (Modify) Bone - Bone: dyn_legR_bowR_MSTR"(ComponentPose=@n63, Alpha=1, Translation="0,0,0", Rotation="-5,45,-45", Scale="1,1,1") {
    rotationMode=BMM_Additive
    rotationSpace=BCS_BoneSpace
    translationMode=BMM_Additive
    translationSpace=BCS_BoneSpace
    alphaBoolBlend {blendOption=Linear}
    boneToModify {boneName=dyn_legR_bowR_MSTR}
    rotation=<rot -5 45 -45>
   }
   Output Pose(Result=Component To Local(ComponentPose=@n64)) {layerGroup=None}
  ]}
 EventGraph {nodes [
   Event BlueprintUpdateAnimation()
   > n1 = Sequence!(execute) -> (then_0, then_1, then_2:Exec, then_3:Exec)
   n2 = Get bIsMoving2D()
   > Set rightLegAlpha(rightLegAlpha:bool=AND Boolean(A=NOT Boolean(A=@n2), B=Get bIsCrouching())) -> (Output_Get:bool)
   n3 = OR Boolean(A=Get bIsRelaxedLevel1() -> (bIsRelaxedLevel1:bool), B=Get bIsRelaxedLevel2() -> (bIsRelaxedLevel2:bool))
   Set necklaceAlpha(execute=@n1.then_1, necklaceAlpha:double=float - float(A=1, B="To Float (Boolean)"(InBool=@n3))) -> (Output_Get:double)
   Set hipStrapAlpha(execute=@n1.then_2, hipStrapAlpha:bool=AND Boolean(A=Get bIsCrouching(), B=Get bIsMoving2D())) -> (Output_Get:bool)
   Set notCrouching(execute=@n1.then_3, notCrouching:bool=NOT Boolean(A=Get bIsCrouching())) -> (Output_Get:bool)
  ]}
]
variables [
 hipStrapAlpha {type=boolean}
 necklaceAlpha {type=number}
 notCrouching {type=boolean}
 rightLegAlpha {type=boolean}
]
